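# CorvidStomp_CapeColliders is a physics asset of collision shapes for the CorvidStomp male medium player body, named as colliders for its cape.
# Each body is a single capsule on a pelvis, spine, arm, hand, leg or foot bone, with no head or neck shapes.
# It has no constraints, so the capsules form a collision proxy that follows the bones rather than a jointed ragdoll.
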PhysicsAsset /BRCosmetics/Characters/Player/Male/Medium/Bodies/M_MED_CorvidStomp/Meshes/CorvidStomp_CapeColliders
bodies {
 calf_l [{shapeType=Capsule, length=28.5842, radius=14.0408, center=<-18.4157 -2.2427 0>, rotation=<rot 90 0 0>}]
 calf_r [{shapeType=Capsule, length=28.5842, radius=14.0408, center=<18.4157 2.2427 0>, rotation=<rot 90 0 180>}]
 foot_l [{shapeType=Capsule, length=15.0766, radius=10.0396, center=<-3.9164 3.986 0>, rotation=<rot 0 14.9998 90>}]
 foot_r [{shapeType=Capsule, length=15.0766, radius=10.0396, center=<3.9164 -3.986 0>, rotation=<rot 0 14.9998 -90>}]
 hand_l [{shapeType=Capsule, length=9.5015, radius=6.7933, center=<6.2703 0 0>, rotation=<rot 90 0 0>}]
 hand_r [{shapeType=Capsule, length=9.5015, radius=6.7933, center=<-6.2703 0 0>, rotation=<rot 90 0 180>}]
 lowerarm_l [{shapeType=Capsule, length=12.2721, radius=9.624, center=<8.9852 0 1.9278>, rotation=<rot 90 0 0>}]
 lowerarm_r [{shapeType=Capsule, length=12.2721, radius=9.624, center=<-8.9852 0 -1.9278>, rotation=<rot 90 0 180>}]
 pelvis [{shapeType=Capsule, length=13.5343, radius=15.539, center=<-7.4686 0.122 0>}]
 spine_02 [{shapeType=Capsule, length=11.7541, radius=16.7458, center=<-6.7616 0.7104 0>}]
 spine_03 [{shapeType=Capsule, length=10, radius=16.0565}]
 spine_04 [{shapeType=Capsule, length=13.6023, radius=17.9115, center=<3.3429 2.3591 0>}]
 spine_05 [{
   shapeType=Capsule
   length=23.8171
   radius=17.1122
   shapeName=spine_05_capsule
   center=<3.1949 1.7789 1.6211>
   rotation=<rot -0.7346 -89.198 0.9598>
  }]
 thigh_l [{shapeType=Capsule, length=15.4343, radius=14.6208, center=<-22.1102 0 -0.1473>, rotation=<rot 90 0 180>}]
 thigh_r [{shapeType=Capsule, length=15.2203, radius=14.8348, center=<22.1092 0.8027 -0.4594>, rotation=<rot 90 165.1766 165.1066>}]
 upperarm_l [{shapeType=Capsule, length=23.4535, radius=10.9678, center=<13.1862 0 0>, rotation=<rot 90 0 0>}]
 upperarm_r [{shapeType=Capsule, length=23.2035, radius=10.7178, center=<-12.9664 1.5303 0>, rotation=<rot 90 0 180>}]
}
constraints []
