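# Control Rig for the CorvidStomp medium male body cosmetic, adding corrective motion to its shoulder armor, hip bags and pants.
# A shared DeformRig function runs first, then a sequence handles shoulders, bags, pants back bones, and calf constraints on pants side and front bones.
# Weighted parent constraints place the armor and bag bones.
# Additive offsets are weighted by deformRig anim attributes read from root, spherical pose readers on upperarm, pelvis and spine, or their products.
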ControlRigBlueprint /BRCosmetics/Characters/Player/Male/Medium/Bodies/M_MED_CorvidStomp/Meshes/M_MED_CorvidStomp_CtrlRig
graphs [
 ContainedGraph {nodes [
   n1 = RigVMFunctionEntryNode()
   SequenceExecution_3_1 = &RigVMFunction_Sequence(ExecuteContext=@n1.ExecuteContext)
   SequenceExecution_3_1_1 = &RigVMFunction_Sequence(ExecuteContext=@SequenceExecution_3_1.B)
   SequenceExecution_3_1_2 = &RigVMFunction_Sequence(ExecuteContext=@SequenceExecution_3_1_1.B)
   RigVMFunctionReturnNode(A=@SequenceExecution_3_1.A, B=@SequenceExecution_3_1_1.A, C=@SequenceExecution_3_1_2.A, D=@SequenceExecution_3_1_2.B)
  ]}
 Multiply_1_ContainedGraph {nodes [
   n1 = RigVMFunctionEntryNode()
   Multiply = &RigVMFunction_MathDoubleMul(A=@n1.A, B=@n1.B)
   RigVMFunctionReturnNode(Result=&RigVMFunction_MathDoubleMul(B=@n1.C, A=@Multiply.Result).Result)
  ]}
 RigVMFunctionLibrary {links [], nodes []}
 RigVMModel {nodes [
   RigUnit_BeginExecution = &RigUnit_BeginExecution()
   SequenceExecution_3_1 = RigVMAggregateNode(ExecuteContext=@RigUnit_BeginExecution.ExecutePin)
   FN_DeformRig_FNC = RigVMFunctionReferenceNode(
    DeformRig_ Curve="(deformRig_l_upperarm_up_alpha_72_#1=false,deformRig_r_upperarm_up_alpha_73_#2=true,deformRig_l_upperarm_fr_alpha_74_#3=false,deformRig_r_upperarm_fr_alpha_75_CB51B80 ...(truncated)"
    DeformRig_Bool="(upperarm_1_#4=true,elbow_3_#5=True,wrist_7_#6=True,groin_8_#7=True,glute_17_#8=True,knee_9_40C408324CA261 ...(truncated)"
    DeformRig_Debug=(upperarm_debug_18_#4=false,wrist_debug_20_#6=false,lat_debug_26_#9=false,groinGlute_debug_27_#7=false)
    DeformRig_Value="(upperarm_up_18_#10=(Rotation=(X=0,Y=0,Z=0,W=1),Translation=(X=0,Y=0,Z=8),Scale3D=(X=0.7,Y=1.0,Z=1.0)),upperarm_fwd_fr_19_#11=(Rotation=(X=0.0,Y=-0.0,Z=0.0,W=1.0),Translation=(X=0,Y=10,Z=0.0),Scale3 ...(truncated)"
    ExecuteContext=@SequenceExecution_3_1.A
   )
   > &RigUnit_ParentConstraint(
    AdvancedSettings=(RotationOrderForFilter=XZY)
    Child=(Type=Bone,Name="right_shoulderplate_MSTR_jnt")
    Parents="((Item=(Type=Bone,Name=\"spine_04\"),Weight=0.0077),(Item=(Type=Bone,Name=\"upperarm_twist_01_r\"),Weight=0.0102),(Item=(Type=Bone,Name=\"spine_03\"),Weight=0.031),(Item=(Type=Bone,Name=\"upperarm_twist_02_r\"),Weight=0.0409),(Item=(Type=Bone,Name=\"deform_upperarm ...(truncated)"
    Weight=1
   )
   > &RigUnit_ParentConstraint(
    AdvancedSettings=(RotationOrderForFilter=XZY)
    Child=(Type=Bone,Name="left_shldr_hinge_offset_jnt01")
    Parents="((Item=(Type=Bone,Name=\"upperarm_twist_01_l\"),Weight=0.0129),(Item=(Type=Bone,Name=\"spine_04\"),Weight=0.0142),(Item=(Type=Bone,Name=\"upperarm_twist_02_l\"),Weight=0.021),(Item=(Type=Bone,Name=\"deform_upperarm_fr_l\"),Weight=0.0323),(Item=(Type=Bone,Name=\"spi ...(truncated)"
    Weight=1
   )
   > &RigUnit_ParentConstraint(
    AdvancedSettings=(RotationOrderForFilter=XZY)
    Child=(Type=Bone,Name="right_left_hinge_offset_jnt02")
    Parents="((Item=(Type=Bone,Name=\"spine_04\"),Weight=0.003),(Item=(Type=Bone,Name=\"spine_03\"),Weight=0.0279),(Item=(Type=Bone,Name=\"deform_upperarm_fr_l\"),Weight=0.047),(Item=(Type=Bone,Name=\"upperarm_twist_02_l\"),Weight=0.0555),(Item=(Type=Bone,Name=\"upperarm_twist_ ...(truncated)"
    Weight=1
   )
   n1 = RigVMDispatchNode(BoneName=root, Default=0, Name=deformRig_r_pec_alpha) {template="DISPATCH_RigDispatch_GetAnimAttribute(in Name,in BoneName,in Default,out Value,out Found)"}
   > /Script/ControlRig.RigUnit_SetRelativeTransformForItem(
    Child=(Type=Bone,Name="right_shoulderplate_offset_jnt")
    Parent=(Type=Bone,Name="right_shoulderplate_offset_jnt")
    Value=(Rotation=(X=0,Y=-0.0346,Z=0,W=0.9994),Translation=(X=-2.266,Y=-1.5115,Z=0.3076),Scale3D=(X=1,Y=1,Z=1))
    bParentInitial=False
    bPropagateToChildren=True
    Weight=@n1.Value
   ) {
    is=struct RigVMUnitNode
    template="Set Relative Transform::Execute(in Child,in Parent,in bParentInitial,in Value,in Weight,in bPropagateToChildren)"
   }
   n2 = RigVMDispatchNode(BoneName=root, Default=0, Name=deformRig_r_upperarm_up_alpha) {template="DISPATCH_RigDispatch_GetAnimAttribute(in Name,in BoneName,in Default,out Value,out Found)"}
   > &RigUnit_SetRelativeTransformForItem(
    Child=(Type=Bone,Name="right_shoulderplate_offset_jnt")
    Parent=(Type=Bone,Name="right_shoulderplate_offset_jnt")
    Value=(Rotation=(X=0,Y=-0.1272,Z=0,W=0.9919),Translation=(X=0,Y=0,Z=0),Scale3D=(X=1,Y=1,Z=1))
    bParentInitial=False
    bPropagateToChildren=True
    Weight=@n2.Value
   ) {template="Set Relative Transform::Execute(in Child,in Parent,in bParentInitial,in Value,in Weight,in bPropagateToChildren)"}
   > n3 = &RigUnit_SphericalPoseReader(
    ActiveRegionScaleFactors=(PositiveWidth=1,NegativeWidth=3,PositiveHeight=1.5,NegativeHeight=1.5)
    ActiveRegionSize=0.1
    Debug=(bDrawDebug=false,DebugScale=25,DebugSegments=20,DebugThickness=0.25)
    DriverAxis=(X=-1,Y=0,Z=0)
    DriverItem=(Type=Bone,Name="upperarm_r")
    FalloffRegionScaleFactors=(PositiveWidth=1,NegativeWidth=1,PositiveHeight=1,NegativeHeight=1)
    FalloffSize=0.4
    FlipWidthScaling=true
    OptionalParentItem=(Type=Bone)
    RotationOffset=(X=90,Y=180,Z=200)
   )
   > &RigUnit_SetRelativeTransformForItem(
    Child=(Type=Bone,Name="right_shoulderplate_offset_jnt")
    Parent=(Type=Bone,Name="right_shoulderplate_offset_jnt")
    Value=(Rotation=(X=0.0151,Y=-0.0858,Z=0.173,W=0.9811),Translation=(X=0,Y=1.5,Z=1),Scale3D=(X=1,Y=1,Z=1))
    bParentInitial=False
    bPropagateToChildren=True
    Weight=@n3.OutputParam
   ) {template="Set Relative Transform::Execute(in Child,in Parent,in bParentInitial,in Value,in Weight,in bPropagateToChildren)"}
   > &RigUnit_SetRelativeTransformForItem(
    Child=(Type=Bone,Name="r_Upper_shoulderplate_offset_01_jnt")
    Parent=(Type=Bone,Name="r_Upper_shoulderplate_offset_01_jnt")
    Value=(Rotation=(X=0.0858,Y=-0.0151,Z=-0.173,W=0.9811),Translation=(X=0,Y=0,Z=1),Scale3D=(X=1,Y=1,Z=1))
    bParentInitial=False
    bPropagateToChildren=True
    Weight=@n3.OutputParam
   ) {template="Set Relative Transform::Execute(in Child,in Parent,in bParentInitial,in Value,in Weight,in bPropagateToChildren)"}
   > &RigUnit_SetRelativeTransformForItem(
    Child=(Type=Bone,Name="right_shoulderplate_offset_jnt")
    Parent=(Type=Bone,Name="right_shoulderplate_offset_jnt")
    Value=(Rotation=(X=0,Y=-0.1305,Z=0,W=0.9914),Translation=(X=-5,Y=0,Z=-5),Scale3D=(X=1,Y=1,Z=1))
    bParentInitial=False
    bPropagateToChildren=True
    Weight=&RigVMFunction_MathDoubleMul(
     A=&RigVMFunction_MathDoubleMul(B=@n2.Value, A=@n1.Value).Result
     B=&RigVMFunction_MathDoubleRemap(SourceMaximum=1, SourceMinimum=0, TargetMaximum=0, TargetMinimum=1, bClamp=False, Value=@n3.OutputParam).Result
    ).Result
   ) {template="Set Relative Transform::Execute(in Child,in Parent,in bParentInitial,in Value,in Weight,in bPropagateToChildren)"}
   &RigUnit_ParentConstraint(
    AdvancedSettings=(RotationOrderForFilter=XZY)
    Child=(Type=Bone,Name="bag_left_MSTR_jnt")
    Parents=((Item=(Type=Bone,Name="spine_04"),Weight=0.0054),(Item=(Type=Bone,Name="spine_03"),Weight=0.1267),(Item=(Type=Bone,Name="spine_01"),Weight=0.177),(Item=(Type=Bone,Name="spine_02"),Weight=0.2124),(Item=(Type=Bone,Name="pelvis"),Weight=0.4784))
    Weight=1
    ExecutePin=@SequenceExecution_3_1.B
   )
   > &RigUnit_ParentConstraint(
    AdvancedSettings=(RotationOrderForFilter=XZY)
    Child=(Type=Bone,Name="bag_right_MSTR_jnt")
    Parents=((Item=(Type=Bone,Name="spine_04"),Weight=0.0054),(Item=(Type=Bone,Name="spine_03"),Weight=0.1267),(Item=(Type=Bone,Name="spine_01"),Weight=0.177),(Item=(Type=Bone,Name="spine_02"),Weight=0.2124),(Item=(Type=Bone,Name="pelvis"),Weight=0.4784))
    Weight=1
   )
   > n4 = &RigUnit_SphericalPoseReader(
    ActiveRegionScaleFactors=(PositiveWidth=1.5,NegativeWidth=1.5,PositiveHeight=1.5,NegativeHeight=1.5)
    ActiveRegionSize=0.09
    Debug=(bDrawDebug=false,DebugScale=25,DebugSegments=20,DebugThickness=0.25)
    DriverAxis=(X=0,Y=1,Z=0)
    DriverItem=(Type=Bone,Name="pelvis")
    FalloffRegionScaleFactors=(PositiveWidth=1,NegativeWidth=1,PositiveHeight=1,NegativeHeight=1)
    FalloffSize=0.18
    FlipWidthScaling=true
    OptionalParentItem=(Type=Bone)
    RotationOffset=(X=-40,Y=0,Z=180)
   )
   > n5 = &RigUnit_SphericalPoseReader(
    ActiveRegionScaleFactors=(PositiveWidth=0.5,NegativeWidth=0.5,PositiveHeight=1.5,NegativeHeight=1.5)
    ActiveRegionSize=0.09
    Debug=(bDrawDebug=false,DebugScale=25,DebugSegments=20,DebugThickness=0.25)
    DriverAxis=(X=0,Y=1,Z=0)
    DriverItem=(Type=Bone,Name="spine_02")
    FalloffRegionScaleFactors=(PositiveWidth=0.3,NegativeWidth=0.3,PositiveHeight=1,NegativeHeight=1)
    FalloffSize=0.18
    FlipWidthScaling=true
    OptionalParentItem=(Type=Bone,Name="pelvis")
    RotationOffset=(X=-100,Y=0,Z=178)
   )
   > n6 = &RigUnit_SphericalPoseReader(
    ActiveRegionScaleFactors=(PositiveWidth=0.5,NegativeWidth=0.5,PositiveHeight=1.5,NegativeHeight=1.5)
    ActiveRegionSize=0.09
    Debug=(bDrawDebug=false,DebugScale=25,DebugSegments=20,DebugThickness=0.25)
    DriverAxis=(X=0,Y=1,Z=0)
    DriverItem=(Type=Bone,Name="pelvis")
    FalloffRegionScaleFactors=(PositiveWidth=0.3,NegativeWidth=0.3,PositiveHeight=1,NegativeHeight=1)
    FalloffSize=0.18
    FlipWidthScaling=true
    OptionalParentItem=(Type=Bone,Name="None")
    RotationOffset=(X=-100,Y=0,Z=180)
   )
   n7 = RigVMDispatchNode(BoneName=root, Default=0, Name=deformRig_l_thigh_up_alpha) {template="DISPATCH_RigDispatch_GetAnimAttribute(in Name,in BoneName,in Default,out Value,out Found)"}
   > &RigUnit_SetRelativeTransformForItem(
    Child=(Type=Bone,Name="bag_left_01_jnt")
    Parent=(Type=Bone,Name="bag_left_01_jnt")
    Value=(Rotation=(X=-0.0121,Y=0.4307,Z=-0.2413,W=0.8696),Translation=(X=-1,Y=0,Z=0),Scale3D=(X=1,Y=1,Z=1))
    bParentInitial=False
    bPropagateToChildren=True
    Weight=@n7.Value
   ) {template="Set Relative Transform::Execute(in Child,in Parent,in bParentInitial,in Value,in Weight,in bPropagateToChildren)"}
   n8 = RigVMDispatchNode(BoneName=root, Default=0, Name=deformRig_r_thigh_up_alpha) {template="DISPATCH_RigDispatch_GetAnimAttribute(in Name,in BoneName,in Default,out Value,out Found)"}
   > &RigUnit_SetRelativeTransformForItem(
    Child=(Type=Bone,Name="bag_right_01_jnt")
    Parent=(Type=Bone,Name="bag_right_01_jnt")
    Value=(Rotation=(X=0.0594,Y=0.3368,Z=-0.1632,W=0.9254),Translation=(X=0,Y=0,Z=0),Scale3D=(X=1,Y=1,Z=1))
    bParentInitial=False
    bPropagateToChildren=True
    Weight=@n8.Value
   ) {template="Set Relative Transform::Execute(in Child,in Parent,in bParentInitial,in Value,in Weight,in bPropagateToChildren)"}
   Subtract = &RigVMFunction_MathDoubleSub(A=1, B=@n5.OutputParam)
   Multiply = RigVMAggregateNode(C=@n7.Value, B=@n6.OutputParam, A=@Subtract.Result)
   > &RigUnit_SetRelativeTransformForItem(
    Child=(Type=Bone,Name="bag_left_01_jnt")
    Parent=(Type=Bone,Name="bag_left_01_jnt")
    Value=(Rotation=(X=0,Y=0.1305,Z=0,W=0.9914),Translation=(X=0,Y=0,Z=0),Scale3D=(X=1,Y=1,Z=1))
    bParentInitial=False
    bPropagateToChildren=True
    Weight=@Multiply.Result
   ) {template="Set Relative Transform::Execute(in Child,in Parent,in bParentInitial,in Value,in Weight,in bPropagateToChildren)"}
   > &RigUnit_SetRelativeTransformForItem(
    Child=(Type=Bone,Name="bag_right_01_jnt")
    Parent=(Type=Bone,Name="bag_right_01_jnt")
    Value=(Rotation=(X=0,Y=0.0872,Z=0,W=0.9962),Translation=(X=0,Y=0,Z=0),Scale3D=(X=1,Y=1,Z=1))
    bParentInitial=False
    bPropagateToChildren=True
    Weight=&RigVMFunction_MathDoubleMul(A=@n8.Value, B=@n4.OutputParam).Result
   ) {template="Set Relative Transform::Execute(in Child,in Parent,in bParentInitial,in Value,in Weight,in bPropagateToChildren)"}
   Remap = &RigVMFunction_MathDoubleRemap(
    SourceMaximum=1
    SourceMinimum=0.8
    TargetMaximum=1
    TargetMinimum=0
    Value=RigVMDispatchNode(BoneName=root, Default=0, Name=deformRig_l_knee_alpha) {template="DISPATCH_RigDispatch_GetAnimAttribute(in Name,in BoneName,in Default,out Value,out Found)"}.Value
   )
   &RigUnit_SetRelativeTransformForItem(
    Child=(Type=Bone,Name="dyn_pants_l_back_02")
    Parent=(Type=Bone,Name="dyn_pants_l_back_02")
    Value=(Rotation=(X=0,Y=0,Z=-0.1736,W=0.9848),Translation=(X=0,Y=0,Z=0),Scale3D=(X=1,Y=1,Z=1))
    bParentInitial=False
    bPropagateToChildren=True
    Weight=@Remap.Result
    ExecutePin=@SequenceExecution_3_1.C
   ) {template="Set Relative Transform::Execute(in Child,in Parent,in bParentInitial,in Value,in Weight,in bPropagateToChildren)"}
   > &RigUnit_SetRelativeTransformForItem(
    Child=(Type=Bone,Name="dyn_pants_r_back_02")
    Parent=(Type=Bone,Name="dyn_pants_r_back_02")
    Value=(Rotation=(X=0,Y=0,Z=-0.1736,W=0.9848),Translation=(X=0,Y=0,Z=0),Scale3D=(X=1,Y=1,Z=1))
    bParentInitial=False
    bPropagateToChildren=True
    Weight=&RigVMFunction_MathDoubleRemap(
     SourceMaximum=1
     SourceMinimum=0.8
     TargetMaximum=1
     TargetMinimum=0
     Value=RigVMDispatchNode(BoneName=root, Default=0, Name=deformRig_r_knee_alpha) {template="DISPATCH_RigDispatch_GetAnimAttribute(in Name,in BoneName,in Default,out Value,out Found)"}.Value
    ).Result
   ) {template="Set Relative Transform::Execute(in Child,in Parent,in bParentInitial,in Value,in Weight,in bPropagateToChildren)"}
   &RigUnit_ParentConstraint(
    AdvancedSettings=(RotationOrderForFilter=XZY)
    Child=(Type=Bone,Name="dyn_pants_l_side_02")
    Filter=(TranslationFilter=(bX=true,bY=true,bZ=true),RotationFilter=(bX=True,bY=True,bZ=True),ScaleFilter=(bX=True,bY=True,bZ=True))
    Parents=((Item=(Type=Bone,Name="calf_l"),Weight=1))
    Weight=1
    ExecutePin=@SequenceExecution_3_1.D
   )
   > &RigUnit_ParentConstraint(
    AdvancedSettings=(RotationOrderForFilter=XZY)
    Child=(Type=Bone,Name="dyn_pants_r_side_02")
    Filter=(TranslationFilter=(bX=true,bY=true,bZ=true),RotationFilter=(bX=True,bY=True,bZ=True),ScaleFilter=(bX=True,bY=True,bZ=True))
    Parents=((Item=(Type=Bone,Name="calf_r"),Weight=1))
    Weight=1
   )
   > &RigUnit_ParentConstraint(
    AdvancedSettings=(RotationOrderForFilter=XZY)
    Child=(Type=Bone,Name="dyn_pants_l_front_02")
    Filter=(TranslationFilter=(bX=true,bY=true,bZ=true),RotationFilter=(bX=True,bY=True,bZ=True),ScaleFilter=(bX=True,bY=True,bZ=True))
    Parents=((Item=(Type=Bone,Name="calf_l"),Weight=1))
    Weight=0.8
   )
   > &RigUnit_ParentConstraint(
    AdvancedSettings=(RotationOrderForFilter=XZY)
    Child=(Type=Bone,Name="dyn_pants_r_front_02")
    Filter=(TranslationFilter=(bX=true,bY=true,bZ=true),RotationFilter=(bX=True,bY=True,bZ=True),ScaleFilter=(bX=True,bY=True,bZ=True))
    Parents=((Item=(Type=Bone,Name="calf_r"),Weight=1))
    Weight=0.8
   )
  ]}
]
hierarchy [
 root {loc=<0 0 0>, rot=<0 0 0 1>}
 pelvis {parent=root, loc=<0 -0.7071 93.1191>, rot=<0.0127 0.707 -0.0127 -0.707>}
 spine_01 {parent=pelvis, loc=<6.7494 -0.6388 0>, rot=<0 0 0.034 -0.9994>}
 spine_02 {parent=spine_01, loc=<7.252 -0.035 0>, rot=<0 0 0.0238 -0.9997>}
 spine_03 {parent=spine_02, loc=<6.0106 0 0>, rot=<0 0 0.0149 -0.9999>}
 spine_04 {parent=spine_03, loc=<8.1023 0 0>, rot=<0 0 -0.0132 -0.9999>}
 spine_05 {parent=spine_04, loc=<10.0142 0 0>, rot=<0 0 -0.0015 -1>}
 attach_backpack {parent=spine_05, loc=<-0.123 -13.473 0>, rot=<-0.0093 -0.7071 0.0094 -0.707>}
 neck_01 {parent=spine_05, loc=<11.3138 0.1529 0>, rot=<0 0 -0.2003 -0.9797>}
 neck_02 {parent=neck_01, loc=<4.5627 0 0>, rot=<0 0 0.0504 -0.9987>}
 head {parent=neck_02, loc=<5.3914 0 0>, rot=<0 0 0.115 -0.9934>}
 clavicle_l {parent=spine_05, loc=<4.3575 -0.1466 -2.3223>, rot=<0.0465 -0.7289 -0.009 -0.6829>}
 upperarm_l {parent=clavicle_l, loc=<15.5111 0 0>, rot=<0.0226 -0.3234 -0.0088 -0.9459>}
 lowerarm_l {parent=upperarm_l, loc=<26.6514 0 0>, rot=<0.0391 -0.0099 -0.2453 -0.9686>}
 hand_l {parent=lowerarm_l, loc=<22.3491 0 0>, rot=<-0.7081 -0.0507 0.0975 -0.6975>}
 thumb_01_l {parent=hand_l, loc=<1.3118 -0.3191 -2.459>, rot=<0.6546 -0.4042 0.0664 -0.6354>}
 thumb_02_l {parent=thumb_01_l, loc=<5.9336 -0.0448 0.0401>, rot=<-0.0037 0.0213 0.298 -0.9543>}
 thumb_03_l {parent=thumb_02_l, loc=<2.2223 0.3682 -0.0907>, rot=<0.0118 -0.013 0.0066 -0.9998>}
 index_metacarpal_l {parent=hand_l, loc=<1.6535 0.0183 -2.96>, rot=<0.0158 -0.0124 -0.0173 -0.9996>}
 index_01_l {parent=index_metacarpal_l, loc=<8.3008 0.0357 -0.0524>, rot=<-0.0044 0.0485 0.0924 -0.9945>}
 index_02_l {parent=index_01_l, loc=<3.3587 -0.3004 0.0452>, rot=<-0.002 0.0027 0.3225 -0.9466>}
 index_03_l {parent=index_02_l, loc=<2.5578 -0.123 -0.0252>, rot=<-0.002 0.0067 0.2558 -0.9667>}
 middle_metacarpal_l {parent=hand_l, loc=<2.1856 -0.1725 -0.5984>, rot=<0.006 0.0168 -0.0329 -0.9993>}
 middle_01_l {parent=middle_metacarpal_l, loc=<7.3275 -0.2447 0.0129>, rot=<-0.0208 0.0042 0.0858 -0.9961>}
 middle_02_l {parent=middle_01_l, loc=<4.0123 0.1401 0.1175>, rot=<-0.0218 0.0258 0.3512 -0.9357>}
 middle_03_l {parent=middle_02_l, loc=<2.6077 0.0534 0.1373>, rot=<0.0082 0.0188 0.1708 -0.9851>}
 ring_metacarpal_l {parent=hand_l, loc=<2.2074 -0.3625 1.3638>, rot=<-0.0284 0.0722 -0.0458 -0.9959>}
 ring_01_l {parent=ring_metacarpal_l, loc=<7.0295 -0.6197 -0.1446>, rot=<-0.0489 -0.0252 0.1892 -0.9804>}
 ring_02_l {parent=ring_01_l, loc=<3.6174 0.4647 -0.1027>, rot=<0 0 0.2806 -0.9598>}
 ring_03_l {parent=ring_02_l, loc=<2.4152 0.114 0.0199>, rot=<0 0 0.1189 -0.9929>}
 pinky_metacarpal_l {parent=hand_l, loc=<1.7616 -0.9753 3.3814>, rot=<-0.0397 0.0814 -0.069 -0.9935>}
 pinky_01_l {parent=pinky_metacarpal_l, loc=<7.1709 -0.4831 -0.1094>, rot=<-0.0654 -0.0035 0.2242 -0.9723>}
 pinky_02_l {parent=pinky_01_l, loc=<2.8274 0.0893 -0.1043>, rot=<0.0037 0.0011 0.2854 -0.9584>}
 pinky_03_l {parent=pinky_02_l, loc=<2.2046 0.0307 -0.0078>, rot=<-0.0002 0.0098 0.1445 -0.9895>}
 deform_wrist_dn_base_l {parent=hand_l, loc=<0 -2.2342 0>, rot=<0 0 0 -1>}
 deform_wrist_dn_l {parent=deform_wrist_dn_base_l, loc=<0 0 0>, rot=<0 0 0 1>}
 deform_wrist_up_base_l {parent=hand_l, loc=<0 2.2682 0>, rot=<0 0 0 -1>}
 deform_wrist_up_l {parent=deform_wrist_up_base_l, loc=<0 0 0>, rot=<0 0 0 1>}
 lowerarm_twist_02_l {parent=lowerarm_l, loc=<9.6199 0 0>, rot=<0 0 0 -1>}
 lowerarm_twist_01_l {parent=lowerarm_l, loc=<17.8549 0 0>, rot=<0 0 0 -1>}
 deform_elbow_in_base_l {parent=lowerarm_l, loc=<0.2112 5.259 0.8274>, rot=<0 0 0 -1>}
 deform_elbow_in_l {parent=deform_elbow_in_base_l, loc=<0 0 0>, rot=<0 0 0 1>}
 deform_elbow_out_base_l {parent=lowerarm_l, loc=<0 -4 0>, rot=<0 0 0 -1>}
 deform_elbow_out_l {parent=deform_elbow_out_base_l, loc=<0 0 0>, rot=<0 0 0 1>}
 upperarm_twist_01_l {parent=upperarm_l, loc=<0.0041 0 0>, rot=<0 0 0 -1>}
 deform_upperarm_bk_base_l {parent=upperarm_twist_01_l, loc=<1.0813 -7.2193 -1.9453>, rot=<0 0 0 -1>}
 deform_upperarm_bk_l {parent=deform_upperarm_bk_base_l, loc=<0 0 0>, rot=<0 0 0 1>}
 deform_upperarm_fr_base_l {parent=upperarm_twist_01_l, loc=<-0.3691 7.0506 -2.765>, rot=<0 0 0 -1>}
 deform_upperarm_fr_l {parent=deform_upperarm_fr_base_l, loc=<0 0 0>, rot=<0 0 0 1>}
 deform_upperarm_up_base_l {parent=upperarm_twist_01_l, loc=<-3.2216 -1.1391 3.3311>, rot=<0 0 0 -1>}
 deform_upperarm_up_l {parent=deform_upperarm_up_base_l, loc=<0 0 0>, rot=<0 0 0 1>}
 upperarm_twist_02_l {parent=upperarm_l, loc=<13.7941 0 0>, rot=<0 0 0 -1>}
 left_shldr_hinge_offset_jnt01 {parent=clavicle_l, loc=<13.2771 -1.1092 10.0461>, rot=<0.0008 0.0336 -0.0252 -0.9991>}
 dyn_left_shldr_hinge_jnt01 {parent=left_shldr_hinge_offset_jnt01, loc=<0 0 0>, rot=<0 0 0 1>}
 right_left_hinge_offset_jnt02 {parent=clavicle_l, loc=<16.4224 -1.3282 8.722>, rot=<0.0008 0.0336 -0.0252 -0.9991>}
 dyn_left_hinge_jnt02 {parent=right_left_hinge_offset_jnt02, loc=<0 0 0>, rot=<0 0 0 1>}
 attach_cape {parent=spine_05, loc=<9.017 -11.714 0>, rot=<0.0616 -0.7044 -0.0615 -0.7044>}
 clavicle_r {parent=spine_05, loc=<4.3575 -0.1466 2.3223>, rot=<0.7289 0.0465 -0.6829 0.009>}
 upperarm_r {parent=clavicle_r, loc=<-15.5111 0 0>, rot=<0.0226 -0.3234 -0.0088 -0.9459>}
 lowerarm_r {parent=upperarm_r, loc=<-26.6514 0 0>, rot=<0.0391 -0.0099 -0.2453 -0.9686>}
 hand_r {parent=lowerarm_r, loc=<-22.3491 0 0>, rot=<-0.7081 -0.0507 0.0975 -0.6975>}
 thumb_01_r {parent=hand_r, loc=<-1.3118 0.3191 2.459>, rot=<0.6546 -0.4042 0.0664 -0.6354>}
 thumb_02_r {parent=thumb_01_r, loc=<-5.9336 0.0448 -0.0401>, rot=<-0.0037 0.0213 0.298 -0.9543>}
 thumb_03_r {parent=thumb_02_r, loc=<-2.2223 -0.3682 0.0907>, rot=<0.0118 -0.013 0.0066 -0.9998>}
 index_metacarpal_r {parent=hand_r, loc=<-1.6535 -0.0183 2.96>, rot=<0.0158 -0.0124 -0.0173 -0.9996>}
 index_01_r {parent=index_metacarpal_r, loc=<-8.3008 -0.0357 0.0524>, rot=<-0.0044 0.0485 0.0924 -0.9945>}
 index_02_r {parent=index_01_r, loc=<-3.3587 0.3004 -0.0452>, rot=<-0.002 0.0027 0.3225 -0.9466>}
 index_03_r {parent=index_02_r, loc=<-2.5578 0.123 0.0252>, rot=<-0.002 0.0067 0.2558 -0.9667>}
 middle_metacarpal_r {parent=hand_r, loc=<-2.1856 0.1725 0.5984>, rot=<0.006 0.0168 -0.0329 -0.9993>}
 middle_01_r {parent=middle_metacarpal_r, loc=<-7.3275 0.2447 -0.0129>, rot=<-0.0208 0.0042 0.0858 -0.9961>}
 middle_02_r {parent=middle_01_r, loc=<-4.0123 -0.1401 -0.1175>, rot=<-0.0218 0.0258 0.3512 -0.9357>}
 middle_03_r {parent=middle_02_r, loc=<-2.6077 -0.0534 -0.1373>, rot=<0.0082 0.0188 0.1708 -0.9851>}
 ring_metacarpal_r {parent=hand_r, loc=<-2.2074 0.3625 -1.3638>, rot=<-0.0284 0.0722 -0.0458 -0.9959>}
 ring_01_r {parent=ring_metacarpal_r, loc=<-7.0295 0.6197 0.1446>, rot=<-0.0489 -0.0252 0.1892 -0.9804>}
 ring_02_r {parent=ring_01_r, loc=<-3.6174 -0.4647 0.1027>, rot=<0 0 0.2806 -0.9598>}
 ring_03_r {parent=ring_02_r, loc=<-2.4152 -0.114 -0.0199>, rot=<0 0 0.1189 -0.9929>}
 pinky_metacarpal_r {parent=hand_r, loc=<-1.7616 0.9753 -3.3814>, rot=<-0.0397 0.0814 -0.069 -0.9935>}
 pinky_01_r {parent=pinky_metacarpal_r, loc=<-7.1709 0.4831 0.1094>, rot=<-0.0654 -0.0035 0.2242 -0.9723>}
 pinky_02_r {parent=pinky_01_r, loc=<-2.8274 -0.0893 0.1043>, rot=<0.0037 0.0011 0.2854 -0.9584>}
 pinky_03_r {parent=pinky_02_r, loc=<-2.2046 -0.0307 0.0078>, rot=<-0.0002 0.0098 0.1445 -0.9895>}
 deform_wrist_dn_base_r {parent=hand_r, loc=<-0.0004 2.2345 -0.0001>, rot=<0 0 0 -1>}
 deform_wrist_dn_r {parent=deform_wrist_dn_base_r, loc=<0 0 0>, rot=<0 0 0 1>}
 deform_wrist_up_base_r {parent=hand_r, loc=<0.0003 -2.2684 0.0001>, rot=<0 0 0 -1>}
 deform_wrist_up_r {parent=deform_wrist_up_base_r, loc=<0 0 0>, rot=<0 0 0 1>}
 lowerarm_twist_02_r {parent=lowerarm_r, loc=<-9.6199 0 0>, rot=<0 0 0 -1>}
 lowerarm_twist_01_r {parent=lowerarm_r, loc=<-17.8549 0 0>, rot=<0 0 0 -1>}
 deform_elbow_in_base_r {parent=lowerarm_r, loc=<-0.2111 -5.2591 -0.8275>, rot=<0 0 0 -1>}
 deform_elbow_in_r {parent=deform_elbow_in_base_r, loc=<-0.0001 0 0.0002>, rot=<0 0 0 1>}
 deform_elbow_out_base_r {parent=lowerarm_r, loc=<0 4 -0.0001>, rot=<0 0 0 -1>}
 deform_elbow_out_r {parent=deform_elbow_out_base_r, loc=<0 0 0>, rot=<0 0 0 1>}
 upperarm_twist_01_r {parent=upperarm_r, loc=<-0.0041 0 0>, rot=<0 0 0 -1>}
 deform_upperarm_bk_base_r {parent=upperarm_twist_01_r, loc=<-1.0814 7.2192 1.9456>, rot=<0 0 0 -1>}
 deform_upperarm_bk_r {parent=deform_upperarm_bk_base_r, loc=<0 0 0>, rot=<0 0 0 1>}
 deform_upperarm_fr_base_r {parent=upperarm_twist_01_r, loc=<0.3689 -7.0506 2.7652>, rot=<0 0 0 -1>}
 deform_upperarm_fr_r {parent=deform_upperarm_fr_base_r, loc=<0 0 0>, rot=<0 0 0 1>}
 deform_upperarm_up_base_r {parent=upperarm_twist_01_r, loc=<3.2214 1.1391 -3.3308>, rot=<0 0 0 -1>}
 deform_upperarm_up_r {parent=deform_upperarm_up_base_r, loc=<0 0 0>, rot=<0 0 0 -1>}
 upperarm_twist_02_r {parent=upperarm_r, loc=<-13.7941 0 0>, rot=<0 0 0 -1>}
 right_shoulderplate_MSTR_jnt {parent=clavicle_r, loc=<-15.5111 0 0>, rot=<0.0226 -0.3234 -0.0088 -0.9459>}
 right_shoulderplate_offset_jnt {parent=right_shoulderplate_MSTR_jnt, loc=<6.8381 0.2846 0.0547>, rot=<0.0108 -0.1732 0.0744 -0.982>}
 dyn_shldr_feathers_joint07 {parent=right_shoulderplate_offset_jnt, loc=<0.0144 1.9777 -10.0936>, rot=<0.9345 0.014 0.3555 -0.0134>}
 dyn_shldr_feathers_joint14 {parent=right_shoulderplate_offset_jnt, loc=<7.5139 2.965 -6.9534>, rot=<0.9345 0.014 0.3555 -0.0134>}
 right_shldr_hinge_offset_jnt01 {parent=right_shoulderplate_offset_jnt, loc=<5.4354 1.1914 -8.2618>, rot=<0.9345 0.014 0.3555 -0.0134>}
 dyn_right_shldr_hinge_jnt01 {parent=right_shldr_hinge_offset_jnt01, loc=<0 0 0>, rot=<0 0 0 1>}
 right_shldr_hinge_offset_jnt02 {parent=right_shoulderplate_offset_jnt, loc=<3.2236 1.2824 -9.5905>, rot=<0.9345 0.014 0.3555 -0.0134>}
 dyn_shldr_hinge_jnt02 {parent=right_shldr_hinge_offset_jnt02, loc=<0 0 0>, rot=<0 0 0 1>}
 r_Lower_shoulderplate_offset_02_jnt {parent=right_shoulderplate_offset_jnt, loc=<0 0 0>, rot=<0.9345 0.014 0.3555 -0.0134>}
 r_Lower_shoulderplate_02_jnt {parent=r_Lower_shoulderplate_offset_02_jnt, loc=<0 0 -14.2998>, rot=<0 0 0 1>}
 r_Upper_shoulderplate_offset_01_jnt {parent=right_shoulderplate_offset_jnt, loc=<1.0503 -6.3559 -3.4876>, rot=<0.9345 0.014 0.3555 -0.0134>}
 r_Upper_shoulderplate_01_jnt {parent=r_Upper_shoulderplate_offset_01_jnt, loc=<0 0 0>, rot=<0 0 0 1>}
 dyn_shldr_feathers_joint01 {parent=r_Upper_shoulderplate_01_jnt, loc=<-3.6829 2.6769 -6.3591>, rot=<0.6226 0.0816 0.0705 -0.7751>}
 dyn_shldr_feathers_joint02 {parent=r_Upper_shoulderplate_01_jnt, loc=<-4.1943 2.0035 -3.3581>, rot=<0.4808 0.052 0.0988 -0.8697>}
 dyn_shldr_feathers_joint03 {parent=r_Upper_shoulderplate_01_jnt, loc=<-4.7659 0.361 -0.7079>, rot=<0.4218 0.0296 0.0557 -0.9045>}
 dyn_shldr_feathers_joint04 {parent=r_Upper_shoulderplate_01_jnt, loc=<-5.3849 -1.9901 1.4989>, rot=<0.2841 0 0 -0.9588>}
 dyn_shldr_feathers_joint05 {parent=r_Upper_shoulderplate_01_jnt, loc=<-6.7981 -3.0065 3.2751>, rot=<0.2485 -0.0018 0.1064 -0.9628>}
 dyn_shldr_feathers_joint06 {parent=r_Upper_shoulderplate_01_jnt, loc=<-6.8503 -5.3953 3.5642>, rot=<0.1335 -0.0634 0.0865 -0.9852>}
 dyn_shldr_feathers_joint11 {parent=r_Upper_shoulderplate_01_jnt, loc=<3.9627 -0.4056 3.8285>, rot=<0 0 0 -1>}
 dyn_shldr_feathers_joint12 {parent=r_Upper_shoulderplate_01_jnt, loc=<4.1978 1.8589 1.1549>, rot=<0 0 0 -1>}
 dyn_shldr_feathers_joint13 {parent=r_Upper_shoulderplate_01_jnt, loc=<3.5678 -4.2207 6.3505>, rot=<0 0 0 -1>}
 r_Upper_shoulderplate_offset_02_jnt {parent=right_shoulderplate_offset_jnt, loc=<0.4676 8.3527 -3.6529>, rot=<0.9345 0.014 0.3555 -0.0134>}
 r_Upper_shoulderplate_02_jnt {parent=r_Upper_shoulderplate_offset_02_jnt, loc=<0 0 0>, rot=<0 0 0 1>}
 dyn_shldr_feathers_joint08 {parent=r_Upper_shoulderplate_02_jnt, loc=<-6.1939 3.9408 2.3115>, rot=<-0.1348 -0.0117 0.0487 -0.9896>}
 dyn_shldr_feathers_joint09 {parent=r_Upper_shoulderplate_02_jnt, loc=<-5.0837 -0.038 0.0562>, rot=<-0.2877 0.017 -0.0564 -0.9559>}
 dyn_shldr_feathers_joint10 {parent=r_Upper_shoulderplate_02_jnt, loc=<-3.937 -2.6326 -3.6904>, rot=<-0.5597 -0.0539 -0.1902 -0.8047>}
 dyn_shldr_feathers_joint15 {parent=r_Upper_shoulderplate_02_jnt, loc=<4.3537 1.5898 5.2116>, rot=<0 0 0 -1>}
 dyn_shldr_feathers_joint16 {parent=r_Upper_shoulderplate_02_jnt, loc=<4.4496 -0.9649 3.3745>, rot=<0 0.0912 0 -0.9958>}
 dyn_shldr_feathers_joint17 {parent=r_Upper_shoulderplate_02_jnt, loc=<4.6607 -2.831 0.2467>, rot=<0 0 0 -1>}
 dyn_shldr_feathers_joint18 {parent=r_Upper_shoulderplate_02_jnt, loc=<4.9967 -4.0182 -2.4564>, rot=<0 0 0 -1>}
 chicken_foot_MSTR_jnt {parent=spine_05, loc=<2.5646 12.2333 -0.0145>, rot=<-0.2913 -0.9566 0 0>}
 dyn_chicken_foot_jnt {parent=chicken_foot_MSTR_jnt, loc=<1.0324 0 0>, rot=<0 0 0 1>}
 deform_pec_aim_base_l {parent=spine_04, loc=<7.8728 12.6388 -2.4123>, rot=<0.0102 -0.4602 0.3979 -0.7936>}
 deform_pec_aim_l {parent=deform_pec_aim_base_l, loc=<0 0 0>, rot=<0 0 0 1>}
 deform_pec_base_l {parent=deform_pec_aim_l, loc=<10.766 0 0>, rot=<0 0 0 1>}
 deform_pec_l {parent=deform_pec_base_l, loc=<0.055 0 0>, rot=<0 0 0 1>}
 deform_pec_aim_base_r {parent=spine_04, loc=<7.8723 12.6388 2.4123>, rot=<-0.4602 -0.0102 0.7936 0.3979>}
 deform_pec_aim_r {parent=deform_pec_aim_base_r, loc=<0 0 0>, rot=<0 0 0 1>}
 deform_pec_base_r {parent=deform_pec_aim_r, loc=<-10.7662 -0.0006 -0.0007>, rot=<0 0 0 1>}
 deform_pec_r {parent=deform_pec_base_r, loc=<0 0 0>, rot=<0 0 0 1>}
 right_lower_hinge_MSTR_joint {parent=spine_03, loc=<1.2394 -0.4196 16.27>, rot=<-0.3475 -0.759 -0.4713 -0.2848>}
 dyn_right_lower_hinge_joint01 {parent=right_lower_hinge_MSTR_joint, loc=<3.3841 0 0>, rot=<0 -0.2375 0 -0.9714>}
 dyn_right_lower_hinge_joint02 {parent=dyn_right_lower_hinge_joint01, loc=<3.2752 0 0>, rot=<0 0 0 1>}
 dyn_right_lower_hinge_joint03 {parent=right_lower_hinge_MSTR_joint, loc=<0.3228 0.7969 -2.1309>, rot=<-0.5121 -0.607 0.2477 -0.5549>}
 hook_back_MSTR_jnt {parent=spine_01, loc=<2.5951 -12.7712 0>, rot=<0.0114 -0.707 -0.0114 -0.707>}
 dyn_hook_back__jnt {parent=hook_back_MSTR_jnt, loc=<0 0 -0.6467>, rot=<0 0 0 1>}
 thigh_l {parent=pelvis, loc=<-2.5693 -0.0109 -7.7501>, rot=<0.0012 0.0572 0.0155 -0.9982>}
 calf_l {parent=thigh_l, loc=<-39.0484 0 0>, rot=<0.0006 -0.0161 -0.0391 -0.9991>}
 foot_l {parent=calf_l, loc=<-42.561 0.1077 0.4292>, rot=<0.0418 -0.0455 0.0343 -0.9975>}
 ball_l {parent=foot_l, loc=<-6.1599 13.8264 0.6806>, rot=<0.0001 -0.0001 -0.7063 -0.7079>}
 calf_twist_02_l {parent=calf_l, loc=<-18.8064 -0.0967 0.0397>, rot=<0.0237 0.0095 -0.0035 -0.9997>}
 calf_twist_01_l {parent=calf_l, loc=<-35.813 -0.0967 0.0397>, rot=<0.0237 0.0095 -0.0035 -0.9997>}
 deform_knee_in_base_l {parent=calf_l, loc=<1.298 -5.2957 -0.0496>, rot=<0 0 0 -1>}
 deform_knee_in_l {parent=deform_knee_in_base_l, loc=<0 0 0>, rot=<0 0 0 1>}
 deform_knee_out_base_l {parent=calf_l, loc=<0 4 0>, rot=<0 0 0 -1>}
 deform_knee_out_l {parent=deform_knee_out_base_l, loc=<0 0 0>, rot=<0 0 0 1>}
 dyn_pants_l_back_02 {parent=calf_l, loc=<-1.5038 -7.5539 -0.0482>, rot=<0.001 0.0161 0.0141 -0.9998>}
 dyn_pants_l_back_03 {parent=dyn_pants_l_back_02, loc=<-18.0812 0 0>, rot=<-0.0006 -0.016 -0.0391 -0.9991>}
 thigh_twist_01_l {parent=thigh_l, loc=<-21.6303 0 0>, rot=<0.0013 0 0 -1>}
 deform_groin_base_l {parent=thigh_l, loc=<-2.1908 8.1142 -1.7442>, rot=<0 0 0 -1>}
 deform_groin_l {parent=deform_groin_base_l, loc=<0 0 0>, rot=<0 0 0 1>}
 dyn_pants_l_front_MSTR {parent=thigh_l, loc=<-3.451 7.5669 0.0003>, rot=<0 0 0 -1>}
 dyn_pants_l_front_01 {parent=dyn_pants_l_front_MSTR, loc=<0 0 0>, rot=<0 0 0 1>}
 dyn_pants_l_front_02 {parent=thigh_l, loc=<-14.568 6.5551 -2.9059>, rot=<0.0013 0 0 -1>}
 dyn_pants_l_front_03 {parent=dyn_pants_l_front_02, loc=<-33.9385 0.9645 1.9015>, rot=<-0.0006 -0.016 -0.0391 -0.9991>}
 dyn_pants_l_back_MSTR {parent=thigh_l, loc=<-3.602 -5.831 0.0003>, rot=<0 0 -0.0872 -0.9962>}
 dyn_pants_l_back_01 {parent=dyn_pants_l_back_MSTR, loc=<0 0 0>, rot=<0 0 0.0623 -0.9981>}
 dyn_pants_l_side_01 {parent=thigh_l, loc=<-20.1836 3.4405 6.0173>, rot=<0.9868 -0.0251 -0.0159 -0.1594>}
 dyn_pants_l_side_02 {parent=dyn_pants_l_side_01, loc=<-17.936 0 0>, rot=<0 -0.0069 0 -1>}
 dyn_pants_l_side_03 {parent=dyn_pants_l_side_02, loc=<-17.7907 0 0>, rot=<0 0 0 1>}
 thigh_r {parent=pelvis, loc=<-2.5693 -0.0109 7.7501>, rot=<0.0572 -0.0012 0.9982 0.0155>}
 calf_r {parent=thigh_r, loc=<39.0484 0 0>, rot=<0.0006 -0.0161 -0.0391 -0.9991>}
 foot_r {parent=calf_r, loc=<42.561 -0.1077 -0.4292>, rot=<0.0418 -0.0455 0.0343 -0.9975>}
 ball_r {parent=foot_r, loc=<6.1599 -13.8264 -0.6806>, rot=<0.0001 -0.0001 -0.7063 -0.7079>}
 calf_twist_02_r {parent=calf_r, loc=<18.8064 0.0967 -0.0397>, rot=<0.0237 0.0095 -0.0035 -0.9997>}
 calf_twist_01_r {parent=calf_r, loc=<35.813 0.0967 -0.0397>, rot=<0.0237 0.0095 -0.0035 -0.9997>}
 deform_knee_in_base_r {parent=calf_r, loc=<-1.298 5.2957 0.0496>, rot=<0 0 0 -1>}
 deform_knee_in_r {parent=deform_knee_in_base_r, loc=<0 0 0>, rot=<0 0 0 1>}
 deform_knee_out_base_r {parent=calf_r, loc=<0 -4 0>, rot=<0 0 0 -1>}
 deform_knee_out_r {parent=deform_knee_out_base_r, loc=<0 0 0>, rot=<0 0 0 1>}
 dyn_pants_r_back_02 {parent=calf_r, loc=<1.5037 7.5539 0.0483>, rot=<0.001 0.0161 0.0141 -0.9998>}
 dyn_pants_r_back_03 {parent=dyn_pants_r_back_02, loc=<18.0813 0 -0.0001>, rot=<-0.0006 -0.016 -0.0391 -0.9991>}
 thigh_twist_01_r {parent=thigh_r, loc=<21.6303 0 0>, rot=<0.0013 0 0 -1>}
 deform_groin_base_r {parent=thigh_r, loc=<2.1908 -8.1142 1.7442>, rot=<0 0 0 -1>}
 deform_groin_r {parent=deform_groin_base_r, loc=<0 0 0>, rot=<0 0 0 1>}
 dyn_pants_r_front_MSTR {parent=thigh_r, loc=<3.451 -7.5669 -0.0003>, rot=<0 0 0 -1>}
 dyn_pants_r_front_01 {parent=dyn_pants_r_front_MSTR, loc=<0 0 0>, rot=<0 0 0 1>}
 dyn_pants_r_front_02 {parent=thigh_r, loc=<14.568 -6.5551 2.9059>, rot=<0.0013 0 0 -1>}
 dyn_pants_r_front_03 {parent=dyn_pants_r_front_02, loc=<33.9385 -0.9645 -1.9016>, rot=<-0.0006 -0.016 -0.0391 -0.9991>}
 dyn_pants_r_back_MSTR {parent=thigh_r, loc=<3.602 5.831 -0.0003>, rot=<0 0 -0.0872 -0.9962>}
 dyn_pants_r_back_01 {parent=dyn_pants_r_back_MSTR, loc=<0 0 0>, rot=<0 0 0.0623 -0.9981>}
 dyn_pants_r_side_01 {parent=thigh_r, loc=<20.1837 -3.4405 -6.0173>, rot=<0.9868 -0.0251 -0.0159 -0.1594>}
 dyn_pants_r_side_02 {parent=dyn_pants_r_side_01, loc=<17.936 0 0>, rot=<0 -0.0069 0 -1>}
 dyn_pants_r_side_03 {parent=dyn_pants_r_side_02, loc=<17.7907 0 0>, rot=<0 0 0 1>}
 deform_glute_base_l {parent=pelvis, loc=<-8.8064 -8.8222 -8.4287>, rot=<0.0275 0.0501 -0.4472 -0.8926>}
 deform_glute_l {parent=deform_glute_base_l, loc=<0 0 0>, rot=<0 0 0 1>}
 deform_glute_base_r {parent=pelvis, loc=<-8.8062 -8.8222 8.4287>, rot=<0.0501 -0.0275 0.8926 -0.4472>}
 deform_glute_r {parent=deform_glute_base_r, loc=<0 0 0>, rot=<0 0 0 1>}
 bag_left_MSTR_jnt {parent=pelvis, loc=<13.9296 8.6922 -9.7757>, rot=<0.303 -0.0054 0.0171 -0.9528>}
 bag_left_01_jnt {parent=bag_left_MSTR_jnt, loc=<0 0 0>, rot=<0 0 0 1>}
 pendant_l_MSTR_jnt {parent=bag_left_01_jnt, loc=<-4.4355 3.8476 -0.0185>, rot=<-0.0219 -0.6862 -0.0172 -0.7269>}
 dyn_pendant_l_jnt {parent=pendant_l_MSTR_jnt, loc=<0 0 -0.2955>, rot=<0 0 0 1>}
 bag_right_MSTR_jnt {parent=pelvis, loc=<13.9296 8.6922 9.7757>, rot=<-0.0054 -0.303 0.9528 0.0171>}
 bag_right_01_jnt {parent=bag_right_MSTR_jnt, loc=<0 0 0>, rot=<0 0 0 1>}
 pendant_r_MSTR_jnt {parent=bag_right_01_jnt, loc=<4.4354 -3.8476 0.0185>, rot=<-0.0219 -0.6862 -0.0172 -0.7269>}
 dyn_pendant_r_jnt {parent=pendant_r_MSTR_jnt, loc=<0 0 0.2955>, rot=<0 0 0 1>}
]
