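# Physics asset for the male medium CorvidStomp player cosmetic body, setting up jiggle for its dangling accessories: chicken foot, back hook, two pendants and right lower hinge.
# Capsule bodies cover the pelvis, spine and each accessory's master and dyn joints.
# Angular constraints tie dyn joints to their masters, except hinge joint 02, which follows 01, with twist locked on the hinge and foot.
# Both pendant masters connect to the pelvis with fully free rotation.
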PhysicsAsset /BRCosmetics/Characters/Player/Male/Medium/Bodies/M_MED_CorvidStomp/Meshes/M_MED_CorvidStomp_Physics
bodies {
 chicken_foot_MSTR_jnt [{shapeType=Capsule, length=0.505, radius=0.505, shapeName=chicken_foot_MSTR_jnt_capsule, rotation=<rot 0 -90 0>}]
 dyn_chicken_foot_jnt [{
   shapeType=Capsule
   length=4.5513
   radius=4.3779
   shapeName=dyn_chicken_foot_jnt_capsule
   center=<7.2417 -1.2584 0.0214>
   rotation=<rot -0.5613 -90 1.274>
  }]
 dyn_hook_back__jnt [{shapeType=Capsule, length=4.9574, radius=1.9102, shapeName=dyn_hook_back__jnt_capsule, center=<0 -0.632 -4.5619>}]
 dyn_pendant_l_jnt [{shapeType=Capsule, length=4, radius=2, shapeName=dyn_pendant_r_jnt_capsule, center=<-0.0155 0.2118 -3.431>, rotation=<rot 0 0 -180>}]
 dyn_pendant_r_jnt [{shapeType=Capsule, length=4, radius=2, shapeName=dyn_pendant_r_jnt_capsule, center=<0.0155 -0.2118 3.431>}]
 dyn_right_lower_hinge_joint01 [{
   shapeType=Capsule
   length=2
   radius=2
   shapeName=dyn_right_lower_hinge_joint01_capsule
   center=<-0.1045 -0.0887 0.3211>
   rotation=<rot 90 0 0>
  }]
 dyn_right_lower_hinge_joint02 [{
   shapeType=Capsule
   length=2
   radius=2
   shapeName=dyn_right_lower_hinge_joint02_capsule
   center=<1.3052 0.1481 -0.1277>
   rotation=<rot 90 0 0>
  }]
 dyn_right_lower_hinge_joint03 [{shapeType=Capsule, length=4, radius=2, shapeName=dyn_right_lower_hinge_joint03_capsule, center=<1.7188 0.901 0.7514>}]
 hook_back_MSTR_jnt [{shapeType=Capsule, length=5, radius=4, shapeName=hook_back_MSTR_jnt_capsule}]
 pelvis [{
   shapeType=Capsule
   length=5.6859
   radius=13.2881
   shapeName=pelvis_capsule
   center=<0.6395 -0.103 0.0883>
   rotation=<rot -5.2221 -90 -1.2032>
  }]
 pendant_l_MSTR_jnt [{shapeType=Capsule, length=4.7674, radius=4.2326, shapeName=pendant_r_MSTR_jnt_capsule, rotation=<rot 0 0 -180>}]
 pendant_r_MSTR_jnt [{shapeType=Capsule, length=5, radius=4, shapeName=pendant_r_MSTR_jnt_capsule, center=<-0.1474 0.4301 0>}]
 right_lower_hinge_MSTR_joint [{shapeType=Capsule, length=5, radius=4, shapeName=right_lower_hinge_MSTR_joint_capsule}]
 spine_01 [{shapeType=Capsule, length=10.4395, radius=9.5402, shapeName=spine_01_capsule, center=<0 0.4381 0>}]
 spine_03 [{shapeType=Capsule, length=0.724, radius=12.5827, shapeName=spine_03_capsule, center=<0.6719 -0.4413 0.7113>}]
 spine_05 [{shapeType=Capsule, length=4.6799, radius=15.3201, shapeName=spine_05_capsule, center=<-7.1649 -2.2986 0.101>, rotation=<rot 0 179.9994 -180>}]
}
constraints [
 {
  bone1Name=dyn_right_lower_hinge_joint01
  bone2Name=right_lower_hinge_MSTR_joint
  swing1LimitDegrees=45
  swing2LimitDegrees=45
  twistLimitDegrees=45
  twistMotion=Locked
 }
 {
  bone1Name=dyn_right_lower_hinge_joint02
  bone2Name=dyn_right_lower_hinge_joint01
  swing1LimitDegrees=45
  swing2LimitDegrees=45
  twistLimitDegrees=45
  twistMotion=Locked
 }
 {
  bone1Name=dyn_right_lower_hinge_joint03
  bone2Name=right_lower_hinge_MSTR_joint
  swing1LimitDegrees=45
  swing2LimitDegrees=45
  twistLimitDegrees=45
  twistMotion=Locked
 }
 {bone1Name=dyn_hook_back__jnt, bone2Name=hook_back_MSTR_jnt, swing1LimitDegrees=5, swing2LimitDegrees=30, twistLimitDegrees=35}
 {bone1Name=dyn_pendant_r_jnt, bone2Name=pendant_r_MSTR_jnt, swing1LimitDegrees=5, swing2LimitDegrees=25, twistLimitDegrees=35}
 {
  bone1Name=dyn_chicken_foot_jnt
  bone2Name=chicken_foot_MSTR_jnt
  swing1LimitDegrees=25
  swing2LimitDegrees=10
  twistLimitDegrees=45
  twistMotion=Locked
 }
 {
  bone1Name=dyn_pendant_l_jnt
  bone2Name=pendant_l_MSTR_jnt
  swing1LimitDegrees=15
  swing1Motion=Locked
  swing2LimitDegrees=15
  twistLimitDegrees=15
 }
 {
  bone1Name=pendant_r_MSTR_jnt
  bone2Name=pelvis
  swing1LimitDegrees=45
  swing1Motion=Free
  swing2LimitDegrees=45
  swing2Motion=Free
  twistLimitDegrees=45
  twistMotion=Free
 }
 {
  bone1Name=pendant_l_MSTR_jnt
  bone2Name=pelvis
  swing1LimitDegrees=45
  swing1Motion=Free
  swing2LimitDegrees=45
  swing2Motion=Free
  twistLimitDegrees=45
  twistMotion=Free
 }
]
