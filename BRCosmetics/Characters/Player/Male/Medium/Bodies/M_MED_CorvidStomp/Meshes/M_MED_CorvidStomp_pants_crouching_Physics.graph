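# Physics asset for the crouching variant of the CorvidStomp medium male pants mesh, covering the lower legs and dynamic pant bones.
# Capsules wrap the calves, calf twist bones and front and back pant chains on each leg.
# Each chain's 02 bone hinges off the calf with 03 off 02, while the tiny 01 capsules are unconstrained.
# Swing is limited; front chains lock twist, back chains allow slight twist, and calf twist joints are free.
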PhysicsAsset /BRCosmetics/Characters/Player/Male/Medium/Bodies/M_MED_CorvidStomp/Meshes/M_MED_CorvidStomp_pants_crouching_Physics
bodies {
 calf_l [{shapeType=Capsule, length=17.1882, radius=15.198, shapeName=calf_l_capsule, center=<-17.872 -12.9209 0>, rotation=<rot 90 0 0>}]
 calf_r [{shapeType=Capsule, length=17.1882, radius=15.198, center=<17.872 12.9209 0.0003>, rotation=<rot 90 0 0>}]
 calf_twist_02_l [{shapeType=Capsule, length=37.1288, radius=13.9109, shapeName=calf_twist_02_l_capsule, center=<-4.3532 13.768 1.0115>, rotation=<rot 90 0 0>}]
 calf_twist_02_r [{shapeType=Capsule, length=37.1288, radius=13.9109, shapeName=calf_twist_02_r_capsule, center=<5.6066 -13.768 -1.0115>, rotation=<rot 90 0 0>}]
 dyn_pants_l_back_01 [{shapeType=Capsule, length=0.505, radius=0.505, shapeName=dyn_pants_l_back_01_capsule, rotation=<rot 0 -90 0>}]
 dyn_pants_l_back_02 [{
   shapeType=Capsule
   length=16.1429
   radius=7.4432
   shapeName=dyn_pants_l_back_02_capsule
   center=<-7.7958 0.4253 0.8983>
   rotation=<rot 0 -90 0>
  }]
 dyn_pants_l_back_03 [{
   shapeType=Capsule
   length=16.1429
   radius=7.4432
   shapeName=dyn_pants_l_back_03_capsule
   center=<-7.7635 1.0347 0.6479>
   rotation=<rot 0 -90 0>
  }]
 dyn_pants_l_front_01 [{shapeType=Capsule, length=0.505, radius=0.505, shapeName=dyn_pants_l_front_01_capsule, rotation=<rot 0 -90 0>}]
 dyn_pants_l_front_02 [{shapeType=Capsule, length=9.951, radius=6.2, shapeName=dyn_pants_l_front_02_capsule, center=<-9 3 0>}]
 dyn_pants_l_front_03 [{shapeType=Capsule, length=12.987, radius=6.1146, shapeName=dyn_pants_l_front_03_capsule, center=<-7 1 0>}]
 dyn_pants_r_back_01 [{shapeType=Capsule, length=0.505, radius=0.505, shapeName=dyn_pants_r_back_01_capsule, rotation=<rot 0 -90 0>}]
 dyn_pants_r_back_02 [{
   shapeType=Capsule
   length=16.1429
   radius=7.4432
   shapeName=dyn_pants_r_back_02_capsule
   center=<7.7958 -0.4253 -0.8983>
   rotation=<rot 0 -90 0>
  }]
 dyn_pants_r_back_03 [{
   shapeType=Capsule
   length=16.1429
   radius=7.4432
   shapeName=dyn_pants_r_back_03_capsule
   center=<7.7635 -1.0347 -0.6479>
   rotation=<rot 0 -90 0>
  }]
 dyn_pants_r_front_01 [{shapeType=Capsule, length=0.505, radius=0.505, shapeName=dyn_pants_r_front_01_capsule, rotation=<rot 0 -90 0>}]
 dyn_pants_r_front_02 [{shapeType=Capsule, length=9.951, radius=6.2, center=<9 -3 0>}]
 dyn_pants_r_front_03 [{shapeType=Capsule, length=12.987, radius=6.1146, shapeName=dyn_pants_r_front_03_capsule, center=<7 -1 0>}]
}
constraints [
 {bone1Name=dyn_pants_r_front_02, bone2Name=calf_r, swing1LimitDegrees=25, swing2LimitDegrees=10, twistLimitDegrees=1, twistMotion=Locked}
 {
  bone1Name=dyn_pants_r_front_03
  bone2Name=dyn_pants_r_front_02
  swing1LimitDegrees=45
  swing2LimitDegrees=10
  twistLimitDegrees=1
  twistMotion=Locked
 }
 {
  bone1Name=dyn_pants_l_front_03
  bone2Name=dyn_pants_l_front_02
  swing1LimitDegrees=45
  swing2LimitDegrees=10
  twistLimitDegrees=1
  twistMotion=Locked
 }
 {bone1Name=dyn_pants_l_front_02, bone2Name=calf_l, swing1LimitDegrees=25, swing2LimitDegrees=10, twistLimitDegrees=1, twistMotion=Locked}
 {
  bone1Name=calf_twist_02_r
  bone2Name=calf_r
  swing1LimitDegrees=45
  swing1Motion=Free
  swing2LimitDegrees=45
  swing2Motion=Free
  twistLimitDegrees=45
  twistMotion=Free
 }
 {
  bone1Name=calf_twist_02_l
  bone2Name=calf_l
  swing1LimitDegrees=45
  swing1Motion=Free
  swing2LimitDegrees=45
  swing2Motion=Free
  twistLimitDegrees=45
  twistMotion=Free
 }
 {bone1Name=dyn_pants_r_back_02, bone2Name=calf_r, swing1LimitDegrees=25, swing2LimitDegrees=10, twistLimitDegrees=5}
 {bone1Name=dyn_pants_r_back_03, bone2Name=dyn_pants_r_back_02, swing1LimitDegrees=45, swing2LimitDegrees=10, twistLimitDegrees=5}
 {bone1Name=dyn_pants_l_back_02, bone2Name=calf_l, swing1LimitDegrees=25, swing2LimitDegrees=10, twistLimitDegrees=5}
 {bone1Name=dyn_pants_l_back_03, bone2Name=dyn_pants_l_back_02, swing1LimitDegrees=45, swing2LimitDegrees=10, twistLimitDegrees=5}
]
