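# Post-process style Anim Blueprint for the CorvidStomp male body cosmetic, adding secondary motion to shoulder feathers, hinges and pants.
# It copies the attached parent mesh's pose through a Control Rig and weighted bone constraints, then chains per-bone AnimDynamics around a pelvis-space RigidBody node.
# Each update sets GravityOverride for the RigidBody via a deprecated helper.
# The crouching-pants RigidBody is unconnected, and the computed frontendAlpha and simSpaceChange are unread.
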AnimBlueprint /BRCosmetics/Characters/Player/Male/Medium/Bodies/M_MED_CorvidStomp/Meshes/M_MED_CorvidStomp_AnimBP
target_skeleton=&/M_MED_CorvidStomp_Skeleton
graphs [
 AnimGraph {nodes [
   RigidBody!(ComponentPose, Alpha, OverrideWorldGravity:Vector=Get GravityOverride(), ExternalForce) -> (Pose) {
    bOverrideWorldGravity=true
    overridePhysicsAsset=&/M_MED_CorvidStomp_pants_crouching_Physics
    simulationSpace=BaseBoneSpace
    alphaBoolBlend {blendOption=Linear}
    alphaScaleBiasClamp {bInterpResult=true, interpSpeedIncreasing=1}
    baseBoneRef {boneName=root}
    componentAppliedLinearAccClamp=<2000 2000 2000>
    componentLinearAccScale=<0.1 0.1 0.1>
    componentLinearVelScale=<0.1 0.1 0.1>
    simSpaceSettings {maxAngularAcceleration=1000, maxAngularVelocity=300, maxLinearAcceleration=1000, maxLinearVelocity=50, worldAlpha=1}
   }
   n1 = Control Rig(Source=Copy Pose From Mesh() {bUseAttachedParent=true}) {
    alphaBoolBlend {blendOption=Linear}
    controlRigAssetReference {blueprintRigClass=&/M_MED_CorvidStomp_CtrlRig.M_MED_CorvidStomp_CtrlRig_C}
   }
   n2 = "Constraint - right_lower_hinge_MSTR_joint (PPPPP)"!(
    ComponentPose=Local To Component(LocalPose=@n1)
    Alpha=1
    ConstraintWeights_0=0.3674
    ConstraintWeights_1:float=0.3226
    ConstraintWeights_2:float=0.1872
    ConstraintWeights_3:float=0.1009
    ConstraintWeights_4:float=0.022
   ) -> (Pose) {
    alphaBoolBlend {blendOption=Linear}
    boneToModify {boneName=right_lower_hinge_MSTR_joint}
    constraintSetup [
     {targetBone {boneName=spine_03}}
     {targetBone {boneName=spine_02}}
     {targetBone {boneName=spine_04}}
     {targetBone {boneName=spine_01}}
     {targetBone {boneName=pelvis}}
    ]
    constraintWeights [1,1,1,1,1]
   }
   n3 = "Constraint - hook_back_MSTR_jnt (PPPPP)"!(
    ComponentPose=@n2
    Alpha=1
    ConstraintWeights_0=0.6817
    ConstraintWeights_1:float=0.2828
    ConstraintWeights_2:float=0.0274
    ConstraintWeights_3:float=0.0076
    ConstraintWeights_4:float=0.0004
   ) -> (Pose) {
    alphaBoolBlend {blendOption=Linear}
    boneToModify {boneName=hook_back_MSTR_jnt}
    constraintSetup [
     {targetBone {boneName=pelvis}}
     {targetBone {boneName=spine_01}}
     {targetBone {boneName=spine_03}}
     {targetBone {boneName=spine_02}}
     {targetBone {boneName=thigh_twist_01_r}}
    ]
    constraintWeights [1,1,1,1,1]
   }
   n4 = "Constraint - right_shldr_hinge_offset_jnt02 (PP)"!(ComponentPose=@n3, Alpha=1, ConstraintWeights_0=0.8476, ConstraintWeights_1:float=0.1524) -> (Pose) {
    alphaBoolBlend {blendOption=Linear}
    boneToModify {boneName=right_shldr_hinge_offset_jnt02}
    constraintSetup [{targetBone {boneName=dyn_shldr_hinge_jnt02}}, {targetBone {boneName=right_shoulderplate_offset_jnt}}]
    constraintWeights [1,1]
   }
   n5 = "Constraint - right_shldr_hinge_offset_jnt01 (PPPPPPP)"!(
    ComponentPose=@n4
    Alpha=1
    ConstraintWeights_0=0.693
    ConstraintWeights_1:float=0.1548
    ConstraintWeights_2:float=0.0927
    ConstraintWeights_3:float=0.0301
    ConstraintWeights_4:float=0.0265
    ConstraintWeights_5:float=0.0019
    ConstraintWeights_6:float=0.001
   ) -> (Pose) {
    alphaBoolBlend {blendOption=Linear}
    boneToModify {boneName=right_shldr_hinge_offset_jnt01}
    constraintSetup [
     {targetBone {boneName=right_shoulderplate_offset_jnt}}
     {targetBone {boneName=dyn_shldr_feathers_joint14}}
     {targetBone {boneName=dyn_shldr_feathers_joint07}}
     {targetBone {boneName=spine_03}}
     {targetBone {boneName=spine_05}}
     {targetBone {boneName=upperarm_twist_02_r}}
     {targetBone {boneName=spine_04}}
    ]
    constraintWeights [1,1,1,1,1,1,1]
   }
   n6 = Anim Dynamics - Bone: dyn_shldr_feathers_joint01!(SimulationSpace:AnimPhysSimSpaceType Enum=BoneRelative, ComponentPose=@n5, Alpha=0.25, ExternalForce="0,0,0") -> (Pose) {
    angularDampingOverride=0.98
    angularSpringConstant=6000
    bAngularSpring=true
    bGravityOverrideInSimSpace=true
    bOverrideAngularDamping=true
    bOverrideLinearDamping=true
    bUseGravityOverride=true
    gravityScale=0.1
    lODThreshold=1
    linearDampingOverride=0.98
    linearSpringConstant=700
    numSolverIterationsPostUpdate=4
    numSolverIterationsPreUpdate=16
    simulationSpace=BoneRelative
    alphaBoolBlend {blendOption=Linear}
    boundBone {boneName=dyn_shldr_feathers_joint01}
    componentLinearAccScale=<1 1 1>
    componentLinearVelScale=<1 1 1>
    gravityOverride {z=-200}
    physicsBodyDefinitions [{
      sphereCollisionRadius=0
      boundBone {boneName=dyn_shldr_feathers_joint01}
      boxExtents=<8 8 8>
      constraintSetup {angularLimitsMax=<2 60 20>, angularLimitsMin=<-2 -45 -20>, angularTarget=<1 0 0>}
      localJointOffset=<-5 0 0>
     }]
    relativeSpaceBone {boneName=pelvis}
   }
   n7 = Anim Dynamics - Bone: dyn_shldr_feathers_joint02!(SimulationSpace:AnimPhysSimSpaceType Enum=BoneRelative, ComponentPose=@n6, Alpha=0.25, ExternalForce="0,0,0") -> (Pose) {
    angularDampingOverride=0.98
    angularSpringConstant=6000
    bAngularSpring=true
    bGravityOverrideInSimSpace=true
    bOverrideAngularDamping=true
    bOverrideLinearDamping=true
    bUseGravityOverride=true
    gravityScale=0.1
    lODThreshold=1
    linearDampingOverride=0.98
    linearSpringConstant=700
    numSolverIterationsPostUpdate=4
    numSolverIterationsPreUpdate=16
    simulationSpace=BoneRelative
    alphaBoolBlend {blendOption=Linear}
    boundBone {boneName=dyn_shldr_feathers_joint02}
    componentLinearAccScale=<1 1 1>
    componentLinearVelScale=<1 1 1>
    gravityOverride {z=-200}
    physicsBodyDefinitions [{
      sphereCollisionRadius=0
      boundBone {boneName=dyn_shldr_feathers_joint02}
      boxExtents=<10 10 10>
      constraintSetup {angularLimitsMax=<2 60 20>, angularLimitsMin=<-2 -45 -20>, angularTarget=<1 0 0>}
      localJointOffset=<-8 0 0>
     }]
    relativeSpaceBone {boneName=pelvis}
   }
   n8 = Anim Dynamics - Bone: dyn_shldr_feathers_joint03!(SimulationSpace:AnimPhysSimSpaceType Enum=BoneRelative, ComponentPose=@n7, Alpha=0.25, ExternalForce="0,0,0") -> (Pose) {
    angularDampingOverride=0.98
    angularSpringConstant=6000
    bAngularSpring=true
    bGravityOverrideInSimSpace=true
    bOverrideAngularDamping=true
    bOverrideLinearDamping=true
    bUseGravityOverride=true
    gravityScale=0.1
    lODThreshold=1
    linearDampingOverride=0.98
    linearSpringConstant=700
    numSolverIterationsPostUpdate=4
    numSolverIterationsPreUpdate=16
    simulationSpace=BoneRelative
    alphaBoolBlend {blendOption=Linear}
    boundBone {boneName=dyn_shldr_feathers_joint03}
    componentLinearAccScale=<1 1 1>
    componentLinearVelScale=<1 1 1>
    gravityOverride {z=-200}
    physicsBodyDefinitions [{
      sphereCollisionRadius=0
      boundBone {boneName=dyn_shldr_feathers_joint03}
      boxExtents=<12 12 12>
      constraintSetup {angularLimitsMax=<2 60 20>, angularLimitsMin=<-2 -45 -20>, angularTarget=<1 0 0>}
      localJointOffset=<-10 0 0>
     }]
    relativeSpaceBone {boneName=pelvis}
   }
   n9 = Anim Dynamics - Bone: dyn_shldr_feathers_joint04!(SimulationSpace:AnimPhysSimSpaceType Enum=BoneRelative, ComponentPose=@n8, Alpha=0.25, ExternalForce="0,0,0") -> (Pose) {
    angularDampingOverride=0.98
    angularSpringConstant=6000
    bAngularSpring=true
    bGravityOverrideInSimSpace=true
    bOverrideAngularDamping=true
    bOverrideLinearDamping=true
    bUseGravityOverride=true
    gravityScale=0.1
    lODThreshold=1
    linearDampingOverride=0.98
    linearSpringConstant=700
    numSolverIterationsPostUpdate=4
    numSolverIterationsPreUpdate=16
    simulationSpace=BoneRelative
    alphaBoolBlend {blendOption=Linear}
    boundBone {boneName=dyn_shldr_feathers_joint04}
    componentLinearAccScale=<1 1 1>
    componentLinearVelScale=<1 1 1>
    gravityOverride {z=-200}
    physicsBodyDefinitions [{
      sphereCollisionRadius=0
      boundBone {boneName=dyn_shldr_feathers_joint04}
      boxExtents=<12 12 12>
      constraintSetup {angularLimitsMax=<2 60 20>, angularLimitsMin=<-2 -45 -20>, angularTarget=<1 0 0>}
      localJointOffset=<-10 0 0>
     }]
    relativeSpaceBone {boneName=pelvis}
   }
   n10 = Anim Dynamics - Bone: dyn_shldr_feathers_joint05!(SimulationSpace:AnimPhysSimSpaceType Enum=BoneRelative, ComponentPose=@n9, Alpha=0.25, ExternalForce="0,0,0") -> (Pose) {
    angularDampingOverride=0.98
    angularSpringConstant=6000
    bAngularSpring=true
    bGravityOverrideInSimSpace=true
    bOverrideAngularDamping=true
    bOverrideLinearDamping=true
    bUseGravityOverride=true
    gravityScale=0.1
    lODThreshold=1
    linearDampingOverride=0.98
    numSolverIterationsPostUpdate=4
    numSolverIterationsPreUpdate=16
    simulationSpace=BoneRelative
    alphaBoolBlend {blendOption=Linear}
    boundBone {boneName=dyn_shldr_feathers_joint05}
    componentLinearAccScale=<1 1 1>
    componentLinearVelScale=<1 1 1>
    gravityOverride {z=-200}
    physicsBodyDefinitions [{
      sphereCollisionRadius=0
      boundBone {boneName=dyn_shldr_feathers_joint05}
      boxExtents=<12 12 12>
      constraintSetup {angularLimitsMax=<2 60 20>, angularLimitsMin=<-2 -45 -20>, angularTarget=<1 0 0>}
      localJointOffset=<-10 0 0>
     }]
    relativeSpaceBone {boneName=pelvis}
   }
   n11 = Anim Dynamics - Bone: dyn_shldr_feathers_joint06!(SimulationSpace:AnimPhysSimSpaceType Enum=BoneRelative, ComponentPose=@n10, Alpha=0.25, ExternalForce="0,0,0") -> (Pose) {
    angularDampingOverride=0.98
    angularSpringConstant=6000
    bAngularSpring=true
    bGravityOverrideInSimSpace=true
    bOverrideAngularDamping=true
    bOverrideLinearDamping=true
    bUseGravityOverride=true
    gravityScale=0.1
    lODThreshold=1
    linearDampingOverride=0.98
    numSolverIterationsPostUpdate=4
    numSolverIterationsPreUpdate=16
    simulationSpace=BoneRelative
    alphaBoolBlend {blendOption=Linear}
    boundBone {boneName=dyn_shldr_feathers_joint06}
    componentLinearAccScale=<1 1 1>
    componentLinearVelScale=<1 1 1>
    gravityOverride {z=-200}
    physicsBodyDefinitions [{
      sphereCollisionRadius=0
      boundBone {boneName=dyn_shldr_feathers_joint06}
      boxExtents=<14 14 14>
      constraintSetup {angularLimitsMax=<2 60 20>, angularLimitsMin=<-2 -45 -20>, angularTarget=<1 0 0>}
      localJointOffset=<-10 0 0>
     }]
    relativeSpaceBone {boneName=pelvis}
   }
   n12 = Anim Dynamics - Bone: dyn_shldr_feathers_joint07!(SimulationSpace:AnimPhysSimSpaceType Enum=BoneRelative, ComponentPose=@n11, Alpha=0.25, ExternalForce="0,0,0") -> (Pose) {
    angularDampingOverride=0.98
    angularSpringConstant=6000
    bAngularSpring=true
    bGravityOverrideInSimSpace=true
    bOverrideAngularDamping=true
    bOverrideLinearDamping=true
    bUseGravityOverride=true
    gravityScale=0.1
    lODThreshold=1
    linearDampingOverride=0.98
    numSolverIterationsPostUpdate=4
    numSolverIterationsPreUpdate=16
    simulationSpace=BoneRelative
    alphaBoolBlend {blendOption=Linear}
    boundBone {boneName=dyn_shldr_feathers_joint07}
    componentLinearAccScale=<1 1 1>
    componentLinearVelScale=<1 1 1>
    gravityOverride {z=-200}
    physicsBodyDefinitions [{
      sphereCollisionRadius=0
      boundBone {boneName=dyn_shldr_feathers_joint07}
      boxExtents=<11 11 11>
      constraintSetup {angularLimitsMax=<2 60 20>, angularLimitsMin=<-2 -45 -20>, angularTarget=<1 0 0>}
      localJointOffset=<-10 0 0>
     }]
    relativeSpaceBone {boneName=pelvis}
   }
   n13 = Anim Dynamics - Bone: dyn_shldr_feathers_joint08!(SimulationSpace:AnimPhysSimSpaceType Enum=BoneRelative, ComponentPose=@n12, Alpha=0.25, ExternalForce="0,0,0") -> (Pose) {
    angularDampingOverride=0.98
    angularSpringConstant=6000
    bAngularSpring=true
    bGravityOverrideInSimSpace=true
    bOverrideAngularDamping=true
    bOverrideLinearDamping=true
    bUseGravityOverride=true
    gravityScale=0.1
    lODThreshold=1
    linearDampingOverride=0.98
    numSolverIterationsPostUpdate=4
    numSolverIterationsPreUpdate=16
    simulationSpace=BoneRelative
    alphaBoolBlend {blendOption=Linear}
    boundBone {boneName=dyn_shldr_feathers_joint08}
    componentLinearAccScale=<1 1 1>
    componentLinearVelScale=<1 1 1>
    gravityOverride {z=-200}
    physicsBodyDefinitions [{
      sphereCollisionRadius=0
      boundBone {boneName=dyn_shldr_feathers_joint08}
      boxExtents=<8 8 8>
      constraintSetup {angularLimitsMax=<2 60 20>, angularLimitsMin=<-2 -45 -20>, angularTarget=<1 0 0>}
      localJointOffset=<-10 0 0>
     }]
    relativeSpaceBone {boneName=pelvis}
   }
   n14 = Anim Dynamics - Bone: dyn_shldr_feathers_joint09!(SimulationSpace:AnimPhysSimSpaceType Enum=BoneRelative, ComponentPose=@n13, Alpha=0.25, ExternalForce="0,0,0") -> (Pose) {
    angularDampingOverride=0.98
    angularSpringConstant=6000
    bAngularSpring=true
    bGravityOverrideInSimSpace=true
    bOverrideAngularDamping=true
    bOverrideLinearDamping=true
    bUseGravityOverride=true
    gravityScale=0.1
    lODThreshold=1
    linearDampingOverride=0.98
    numSolverIterationsPostUpdate=4
    numSolverIterationsPreUpdate=16
    simulationSpace=BoneRelative
    alphaBoolBlend {blendOption=Linear}
    boundBone {boneName=dyn_shldr_feathers_joint09}
    componentLinearAccScale=<1 1 1>
    componentLinearVelScale=<1 1 1>
    gravityOverride {z=-200}
    physicsBodyDefinitions [{
      sphereCollisionRadius=0
      boundBone {boneName=dyn_shldr_feathers_joint09}
      boxExtents=<8 8 8>
      constraintSetup {angularLimitsMax=<2 60 20>, angularLimitsMin=<-2 -45 -20>, angularTarget=<1 0 0>}
      localJointOffset=<-8 0 0>
     }]
    relativeSpaceBone {boneName=pelvis}
   }
   n15 = Anim Dynamics - Bone: dyn_shldr_feathers_joint10!(SimulationSpace:AnimPhysSimSpaceType Enum=BoneRelative, ComponentPose=@n14, Alpha=0.25, ExternalForce="0,0,0") -> (Pose) {
    angularDampingOverride=0.98
    angularSpringConstant=6000
    bAngularSpring=true
    bGravityOverrideInSimSpace=true
    bOverrideAngularDamping=true
    bOverrideLinearDamping=true
    bUseGravityOverride=true
    gravityScale=0.1
    lODThreshold=1
    linearDampingOverride=0.98
    numSolverIterationsPostUpdate=4
    numSolverIterationsPreUpdate=16
    simulationSpace=BoneRelative
    alphaBoolBlend {blendOption=Linear}
    boundBone {boneName=dyn_shldr_feathers_joint10}
    componentLinearAccScale=<1 1 1>
    componentLinearVelScale=<1 1 1>
    gravityOverride {z=-200}
    physicsBodyDefinitions [{
      sphereCollisionRadius=0
      boundBone {boneName=dyn_shldr_feathers_joint10}
      boxExtents=<8 8 8>
      constraintSetup {angularLimitsMax=<2 60 20>, angularLimitsMin=<-2 -45 -20>, angularTarget=<1 0 0>}
      localJointOffset=<-5 0 0>
     }]
    relativeSpaceBone {boneName=pelvis}
   }
   n16 = Anim Dynamics - Bone: dyn_shldr_feathers_joint11!(SimulationSpace:AnimPhysSimSpaceType Enum=BoneRelative, ComponentPose=@n15, Alpha=0.25, ExternalForce="0,0,0") -> (Pose) {
    angularDampingOverride=0.98
    angularSpringConstant=8000
    bAngularSpring=true
    bGravityOverrideInSimSpace=true
    bOverrideAngularDamping=true
    bUseGravityOverride=true
    gravityScale=0.1
    lODThreshold=1
    linearDampingOverride=0.98
    linearSpringConstant=700
    numSolverIterationsPostUpdate=4
    numSolverIterationsPreUpdate=16
    simulationSpace=BoneRelative
    alphaBoolBlend {blendOption=Linear}
    boundBone {boneName=dyn_shldr_feathers_joint11}
    componentLinearAccScale=<1 1 1>
    componentLinearVelScale=<1 1 1>
    gravityOverride {z=-980}
    physicsBodyDefinitions [{
      sphereCollisionRadius=0
      boundBone {boneName=dyn_shldr_feathers_joint11}
      boxExtents=<5 5 5>
      constraintSetup {angularLimitsMax=<2 60 20>, angularLimitsMin=<-2 -45 -20>, angularTarget=<1 1 1>}
      localJointOffset=<10 0 0>
     }]
    relativeSpaceBone {boneName=pelvis}
   }
   n17 = Anim Dynamics - Bone: dyn_shldr_feathers_joint12!(SimulationSpace:AnimPhysSimSpaceType Enum=BoneRelative, ComponentPose=@n16, Alpha=0.25, ExternalForce="0,0,0") -> (Pose) {
    angularDampingOverride=0.98
    angularSpringConstant=8000
    bAngularSpring=true
    bGravityOverrideInSimSpace=true
    bOverrideAngularDamping=true
    bUseGravityOverride=true
    gravityScale=0.1
    lODThreshold=1
    linearDampingOverride=0.98
    linearSpringConstant=700
    numSolverIterationsPostUpdate=4
    numSolverIterationsPreUpdate=16
    simulationSpace=BoneRelative
    alphaBoolBlend {blendOption=Linear}
    boundBone {boneName=dyn_shldr_feathers_joint12}
    componentLinearAccScale=<1 1 1>
    componentLinearVelScale=<1 1 1>
    gravityOverride {z=-980}
    physicsBodyDefinitions [{
      sphereCollisionRadius=0
      boundBone {boneName=dyn_shldr_feathers_joint12}
      boxExtents=<5 5 5>
      constraintSetup {angularLimitsMax=<2 60 20>, angularLimitsMin=<-2 -45 -20>, angularTarget=<1 1 1>}
      localJointOffset=<10 0 0>
     }]
    relativeSpaceBone {boneName=pelvis}
   }
   n18 = Anim Dynamics - Bone: dyn_shldr_feathers_joint13!(SimulationSpace:AnimPhysSimSpaceType Enum=BoneRelative, ComponentPose=@n17, Alpha=0.25, ExternalForce="0,0,0") -> (Pose) {
    angularDampingOverride=0.98
    angularSpringConstant=8000
    bAngularSpring=true
    bGravityOverrideInSimSpace=true
    bOverrideAngularDamping=true
    bUseGravityOverride=true
    gravityScale=0.1
    lODThreshold=1
    linearDampingOverride=0.98
    linearSpringConstant=700
    numSolverIterationsPostUpdate=4
    numSolverIterationsPreUpdate=16
    simulationSpace=BoneRelative
    alphaBoolBlend {blendOption=Linear}
    boundBone {boneName=dyn_shldr_feathers_joint13}
    componentLinearAccScale=<1 1 1>
    componentLinearVelScale=<1 1 1>
    gravityOverride {z=-980}
    physicsBodyDefinitions [{
      sphereCollisionRadius=0
      boundBone {boneName=dyn_shldr_feathers_joint13}
      boxExtents=<5 5 5>
      constraintSetup {angularLimitsMax=<2 60 20>, angularLimitsMin=<-2 -45 -20>, angularTarget=<1 1 1>}
      localJointOffset=<10 0 0>
     }]
    relativeSpaceBone {boneName=pelvis}
   }
   n19 = Anim Dynamics - Bone: dyn_shldr_feathers_joint14!(SimulationSpace:AnimPhysSimSpaceType Enum=BoneRelative, ComponentPose=@n18, Alpha=0.25, ExternalForce="0,0,0") -> (Pose) {
    angularDampingOverride=0.98
    angularSpringConstant=8000
    bAngularSpring=true
    bGravityOverrideInSimSpace=true
    bOverrideAngularDamping=true
    bUseGravityOverride=true
    gravityScale=0.1
    lODThreshold=1
    linearDampingOverride=0.98
    linearSpringConstant=700
    numSolverIterationsPostUpdate=4
    numSolverIterationsPreUpdate=16
    simulationSpace=BoneRelative
    alphaBoolBlend {blendOption=Linear}
    boundBone {boneName=dyn_shldr_feathers_joint14}
    componentLinearAccScale=<1 1 1>
    componentLinearVelScale=<1 1 1>
    gravityOverride {z=-980}
    physicsBodyDefinitions [{
      sphereCollisionRadius=0
      boundBone {boneName=dyn_shldr_feathers_joint14}
      boxExtents=<5 5 5>
      constraintSetup {angularLimitsMax=<2 60 20>, angularLimitsMin=<-2 -45 -20>, angularTarget=<1 1 1>}
      localJointOffset=<10 0 0>
     }]
    relativeSpaceBone {boneName=pelvis}
   }
   n20 = Anim Dynamics - Bone: dyn_shldr_feathers_joint15!(SimulationSpace:AnimPhysSimSpaceType Enum=BoneRelative, ComponentPose=@n19, Alpha=0.25, ExternalForce="0,0,0") -> (Pose) {
    angularDampingOverride=0.98
    angularSpringConstant=8000
    bAngularSpring=true
    bGravityOverrideInSimSpace=true
    bOverrideAngularDamping=true
    bUseGravityOverride=true
    gravityScale=0.1
    lODThreshold=1
    linearDampingOverride=0.98
    numSolverIterationsPostUpdate=4
    numSolverIterationsPreUpdate=16
    simulationSpace=BoneRelative
    alphaBoolBlend {blendOption=Linear}
    boundBone {boneName=dyn_shldr_feathers_joint15}
    componentLinearAccScale=<1 1 1>
    componentLinearVelScale=<1 1 1>
    gravityOverride {z=-980}
    physicsBodyDefinitions [{
      sphereCollisionRadius=0
      boundBone {boneName=dyn_shldr_feathers_joint15}
      boxExtents=<5 5 5>
      constraintSetup {angularLimitsMax=<2 60 20>, angularLimitsMin=<-2 -45 -20>, angularTarget=<1 1 1>}
      localJointOffset=<10 0 0>
     }]
    relativeSpaceBone {boneName=pelvis}
   }
   n21 = Anim Dynamics - Bone: dyn_shldr_feathers_joint16!(SimulationSpace:AnimPhysSimSpaceType Enum=BoneRelative, ComponentPose=@n20, Alpha=0.25, ExternalForce="0,0,0") -> (Pose) {
    angularDampingOverride=0.98
    angularSpringConstant=8000
    bAngularSpring=true
    bGravityOverrideInSimSpace=true
    bOverrideAngularDamping=true
    bUseGravityOverride=true
    gravityScale=0.1
    lODThreshold=1
    linearDampingOverride=0.98
    numSolverIterationsPostUpdate=4
    numSolverIterationsPreUpdate=16
    simulationSpace=BoneRelative
    alphaBoolBlend {blendOption=Linear}
    boundBone {boneName=dyn_shldr_feathers_joint16}
    componentLinearAccScale=<1 1 1>
    componentLinearVelScale=<1 1 1>
    gravityOverride {z=-980}
    physicsBodyDefinitions [{
      sphereCollisionRadius=0
      boundBone {boneName=dyn_shldr_feathers_joint16}
      boxExtents=<5 5 5>
      constraintSetup {angularLimitsMax=<2 60 20>, angularLimitsMin=<-2 -45 -20>, angularTarget=<1 1 1>}
      localJointOffset=<10 0 0>
     }]
    relativeSpaceBone {boneName=pelvis}
   }
   n22 = Anim Dynamics - Bone: dyn_shldr_feathers_joint18!(SimulationSpace:AnimPhysSimSpaceType Enum=BoneRelative, ComponentPose=@n21, Alpha=0.25, ExternalForce="0,0,0") -> (Pose) {
    angularDampingOverride=0.98
    angularSpringConstant=8000
    bAngularSpring=true
    bGravityOverrideInSimSpace=true
    bOverrideAngularDamping=true
    bUseGravityOverride=true
    gravityScale=0.1
    lODThreshold=1
    linearDampingOverride=0.98
    linearSpringConstant=700
    numSolverIterationsPostUpdate=4
    numSolverIterationsPreUpdate=16
    simulationSpace=BoneRelative
    alphaBoolBlend {blendOption=Linear}
    boundBone {boneName=dyn_shldr_feathers_joint18}
    componentLinearAccScale=<1 1 1>
    componentLinearVelScale=<1 1 1>
    gravityOverride {z=-980}
    physicsBodyDefinitions [{
      sphereCollisionRadius=0
      boundBone {boneName=dyn_shldr_feathers_joint18}
      boxExtents=<5 5 5>
      constraintSetup {angularLimitsMax=<2 60 20>, angularLimitsMin=<-2 -45 -20>, angularTarget=<1 1 1>}
      localJointOffset=<10 0 0>
     }]
    relativeSpaceBone {boneName=pelvis}
   }
   n23 = Anim Dynamics - Bone: dyn_shldr_feathers_joint17!(SimulationSpace:AnimPhysSimSpaceType Enum=BoneRelative, ComponentPose=@n22, Alpha=0.25, ExternalForce="0,0,0") -> (Pose) {
    angularDampingOverride=0.98
    angularSpringConstant=8000
    bAngularSpring=true
    bGravityOverrideInSimSpace=true
    bOverrideAngularDamping=true
    bUseGravityOverride=true
    gravityScale=0.1
    lODThreshold=1
    linearDampingOverride=0.98
    numSolverIterationsPostUpdate=4
    numSolverIterationsPreUpdate=16
    simulationSpace=BoneRelative
    alphaBoolBlend {blendOption=Linear}
    boundBone {boneName=dyn_shldr_feathers_joint17}
    componentLinearAccScale=<1 1 1>
    componentLinearVelScale=<1 1 1>
    gravityOverride {z=-980}
    physicsBodyDefinitions [{
      sphereCollisionRadius=0
      boundBone {boneName=dyn_shldr_feathers_joint17}
      boxExtents=<5 5 5>
      constraintSetup {angularLimitsMax=<2 60 20>, angularLimitsMin=<-2 -45 -20>, angularTarget=<1 1 1>}
      localJointOffset=<10 0 0>
     }]
    relativeSpaceBone {boneName=pelvis}
   }
   n24 = RigidBody(ComponentPose=@n23, OverrideWorldGravity=Get GravityOverride()) {
    bOverrideWorldGravity=true
    overridePhysicsAsset=&/M_MED_CorvidStomp_Physics
    simulationSpace=BaseBoneSpace
    alphaBoolBlend {blendOption=Linear}
    baseBoneRef {boneName=pelvis}
    componentAppliedLinearAccClamp=<2000 2000 2000>
    componentLinearAccScale=<0.1 0.1 0.1>
    componentLinearVelScale=<0.1 0.1 0.1>
    simSpaceSettings {maxAngularAcceleration=1000, maxAngularVelocity=300, maxLinearAcceleration=1000, maxLinearVelocity=50, worldAlpha=1}
   }
   n25 = Anim Dynamics - Bone: dyn_left_shldr_hinge_jnt01!(ComponentPose=@n24, Alpha=1) -> (Pose) {
    angularDampingOverride=1
    angularSpringConstant=100
    bAngularSpring=true
    bGravityOverrideInSimSpace=true
    bOverrideAngularDamping=true
    bUseGravityOverride=true
    simulationSpace=BoneRelative
    alphaBoolBlend {blendOption=Linear}
    boundBone {boneName=dyn_left_shldr_hinge_jnt01}
    componentAppliedLinearAccClamp=<1000 1000 1000>
    componentLinearAccScale=<0.1 0.1 0.1>
    componentLinearVelScale=<0.1 0.1 0.1>
    physicsBodyDefinitions [{
      boundBone {boneName=dyn_left_shldr_hinge_jnt01}
      boxExtents=<2 2 2>
      constraintSetup {angularLimitsMax=<10 20 20>, angularLimitsMin=<-10 -20 -20>, angularTarget=<1 0 0>}
      localJointOffset=<0 0 1>
     }]
    relativeSpaceBone {boneName=spine_05}
   }
   n26 = Anim Dynamics - Bone: dyn_left_hinge_jnt02!(ComponentPose=@n25, Alpha=1) -> (Pose) {
    angularDampingOverride=1
    angularSpringConstant=100
    bAngularSpring=true
    bGravityOverrideInSimSpace=true
    bOverrideAngularDamping=true
    bUseGravityOverride=true
    simulationSpace=BoneRelative
    alphaBoolBlend {blendOption=Linear}
    boundBone {boneName=dyn_left_hinge_jnt02}
    componentAppliedLinearAccClamp=<1000 1000 1000>
    componentLinearAccScale=<0.1 0.1 0.1>
    componentLinearVelScale=<0.1 0.1 0.1>
    physicsBodyDefinitions [{
      boundBone {boneName=dyn_left_hinge_jnt02}
      boxExtents=<2 2 2>
      constraintSetup {angularLimitsMax=<10 20 20>, angularLimitsMin=<-10 -20 -20>, angularTarget=<1 0 0>}
      localJointOffset=<0 0 1>
     }]
    relativeSpaceBone {boneName=spine_05}
   }
   n27 = Anim Dynamics - Bone: dyn_right_shldr_hinge_jnt01!(ComponentPose=@n26, Alpha=1) -> (Pose) {
    angularDampingOverride=1
    angularSpringConstant=100
    bAngularSpring=true
    bGravityOverrideInSimSpace=true
    bOverrideAngularDamping=true
    bUseGravityOverride=true
    simulationSpace=BoneRelative
    alphaBoolBlend {blendOption=Linear}
    boundBone {boneName=dyn_right_shldr_hinge_jnt01}
    componentAppliedLinearAccClamp=<1000 1000 1000>
    componentLinearAccScale=<0.1 0.1 0.1>
    componentLinearVelScale=<0.1 0.1 0.1>
    physicsBodyDefinitions [{
      boundBone {boneName=dyn_right_shldr_hinge_jnt01}
      boxExtents=<2 2 2>
      constraintSetup {angularLimitsMax=<10 20 20>, angularLimitsMin=<-10 -20 -20>, angularTarget=<1 0 0>}
      localJointOffset=<0 0 1>
     }]
    relativeSpaceBone {boneName=spine_05}
   }
   n28 = Anim Dynamics - Bone: dyn_shldr_hinge_jnt02!(ComponentPose=@n27, Alpha=1) -> (Pose) {
    angularDampingOverride=1
    angularSpringConstant=100
    bAngularSpring=true
    bGravityOverrideInSimSpace=true
    bOverrideAngularDamping=true
    bUseGravityOverride=true
    simulationSpace=BoneRelative
    alphaBoolBlend {blendOption=Linear}
    boundBone {boneName=dyn_shldr_hinge_jnt02}
    componentAppliedLinearAccClamp=<1000 1000 1000>
    componentLinearAccScale=<0.1 0.1 0.1>
    componentLinearVelScale=<0.1 0.1 0.1>
    physicsBodyDefinitions [{
      boundBone {boneName=dyn_shldr_hinge_jnt02}
      boxExtents=<2 2 2>
      constraintSetup {angularLimitsMax=<10 20 20>, angularLimitsMin=<-10 -20 -20>, angularTarget=<1 0 0>}
      localJointOffset=<0 0 1>
     }]
    relativeSpaceBone {boneName=spine_05}
   }
   n29 = Anim Dynamics - Bone: dyn_pants_l_side_01!(SimulationSpace:AnimPhysSimSpaceType Enum, ComponentPose=@n28, Alpha=1) -> (Pose) {
    angularDampingOverride=0.9999
    angularSpringConstant=2000
    bAngularSpring=true
    bOverrideAngularDamping=true
    bUseGravityOverride=true
    gravityScale=0.35
    simulationSpace=RootRelative
    alphaBoolBlend {blendOption=Linear}
    boundBone {boneName=dyn_pants_l_side_01}
    componentAppliedLinearAccClamp=<2500 2500 2500>
    componentLinearAccScale=<1 1 1>
    componentLinearVelScale=<1 1 1>
    gravityOverride {z=-980}
    physicsBodyDefinitions [{
      boundBone {boneName=dyn_pants_l_side_01}
      boxExtents=<23 23 23>
      constraintSetup {angularLimitsMax=<10 3 10>, angularLimitsMin=<-10 -20 -10>, angularTarget=<1 0 0>}
      localJointOffset=<-4 0 0>
     }]
   }
   n30 = Anim Dynamics - Bone: dyn_pants_l_side_02!(SimulationSpace:AnimPhysSimSpaceType Enum, ComponentPose=@n29, Alpha=1) -> (Pose) {
    angularDampingOverride=0.9999
    angularSpringConstant=2000
    bAngularSpring=true
    bOverrideAngularDamping=true
    bUseGravityOverride=true
    gravityScale=0.35
    simulationSpace=RootRelative
    alphaBoolBlend {blendOption=Linear}
    boundBone {boneName=dyn_pants_l_side_02}
    componentAppliedLinearAccClamp=<2500 2500 2500>
    componentLinearAccScale=<1 1 1>
    componentLinearVelScale=<1 1 1>
    gravityOverride {z=-980}
    physicsBodyDefinitions [{
      boundBone {boneName=dyn_pants_l_side_02}
      boxExtents=<23 23 23>
      constraintSetup {angularLimitsMax=<5 3 10>, angularLimitsMin=<-5 -20 -10>, angularTarget=<1 0 0>}
      localJointOffset=<-4 0 0>
     }]
   }
   n31 = Anim Dynamics - Bone: dyn_pants_l_side_03!(SimulationSpace:AnimPhysSimSpaceType Enum, ComponentPose=@n30, Alpha=1) -> (Pose) {
    angularDampingOverride=0.9999
    angularSpringConstant=2000
    bAngularSpring=true
    bOverrideAngularDamping=true
    bUseGravityOverride=true
    gravityScale=0.35
    simulationSpace=RootRelative
    alphaBoolBlend {blendOption=Linear}
    boundBone {boneName=dyn_pants_l_side_03}
    componentAppliedLinearAccClamp=<2500 2500 2500>
    componentLinearAccScale=<1 1 1>
    componentLinearVelScale=<1 1 1>
    gravityOverride {z=-980}
    physicsBodyDefinitions [{
      boundBone {boneName=dyn_pants_l_side_03}
      boxExtents=<23 23 23>
      constraintSetup {angularLimitsMax=<5 3 10>, angularLimitsMin=<-5 -20 -10>, angularTarget=<1 0 0>}
      localJointOffset=<-4 0 0>
     }]
   }
   n32 = Anim Dynamics - Bone: dyn_pants_r_side_01!(SimulationSpace:AnimPhysSimSpaceType Enum, ComponentPose=@n31, Alpha=1) -> (Pose) {
    angularDampingOverride=0.9999
    angularSpringConstant=2000
    bAngularSpring=true
    bOverrideAngularDamping=true
    bUseGravityOverride=true
    gravityScale=0.35
    simulationSpace=RootRelative
    alphaBoolBlend {blendOption=Linear}
    boundBone {boneName=dyn_pants_r_side_01}
    componentAppliedLinearAccClamp=<2500 2500 2500>
    componentLinearAccScale=<1 1 1>
    componentLinearVelScale=<1 1 1>
    gravityOverride {z=-980}
    physicsBodyDefinitions [{
      boundBone {boneName=dyn_pants_r_side_01}
      boxExtents=<23 23 23>
      constraintSetup {angularLimitsMax=<10 3 10>, angularLimitsMin=<-10 -20 -10>, angularTarget=<1 0 0>}
      localJointOffset=<4 0 0>
     }]
   }
   n33 = Anim Dynamics - Bone: dyn_pants_r_side_02!(SimulationSpace:AnimPhysSimSpaceType Enum, ComponentPose=@n32, Alpha=1) -> (Pose) {
    angularDampingOverride=0.9999
    angularSpringConstant=2000
    bAngularSpring=true
    bOverrideAngularDamping=true
    bUseGravityOverride=true
    gravityScale=0.35
    simulationSpace=RootRelative
    alphaBoolBlend {blendOption=Linear}
    boundBone {boneName=dyn_pants_r_side_02}
    componentAppliedLinearAccClamp=<2500 2500 2500>
    componentLinearAccScale=<1 1 1>
    componentLinearVelScale=<1 1 1>
    gravityOverride {z=-980}
    physicsBodyDefinitions [{
      boundBone {boneName=dyn_pants_r_side_02}
      boxExtents=<23 23 23>
      constraintSetup {angularLimitsMax=<5 3 10>, angularLimitsMin=<-5 -20 -10>, angularTarget=<1 0 0>}
      localJointOffset=<4 0 0>
     }]
   }
   n34 = Anim Dynamics - Bone: dyn_pants_r_side_03!(SimulationSpace:AnimPhysSimSpaceType Enum, ComponentPose=@n33, Alpha=1) -> (Pose) {
    angularDampingOverride=0.9999
    angularSpringConstant=2000
    bAngularSpring=true
    bOverrideAngularDamping=true
    bUseGravityOverride=true
    gravityScale=0.35
    simulationSpace=RootRelative
    alphaBoolBlend {blendOption=Linear}
    boundBone {boneName=dyn_pants_r_side_03}
    componentAppliedLinearAccClamp=<2500 2500 2500>
    componentLinearAccScale=<1 1 1>
    componentLinearVelScale=<1 1 1>
    gravityOverride {z=-980}
    physicsBodyDefinitions [{
      boundBone {boneName=dyn_pants_r_side_03}
      boxExtents=<23 23 23>
      constraintSetup {angularLimitsMax=<5 3 10>, angularLimitsMin=<-5 -20 -10>, angularTarget=<1 0 0>}
      localJointOffset=<4 0 0>
     }]
   }
   n35 = Anim Dynamics - Bone: dyn_pants_l_front_02!(SimulationSpace:AnimPhysSimSpaceType Enum, ComponentPose=@n34, Alpha=1) -> (Pose) {
    angularDampingOverride=0.9999
    angularSpringConstant=2000
    bAngularSpring=true
    bOverrideAngularDamping=true
    bUseGravityOverride=true
    gravityScale=0.35
    simulationSpace=RootRelative
    alphaBoolBlend {blendOption=Linear}
    boundBone {boneName=dyn_pants_l_front_02}
    componentAppliedLinearAccClamp=<2500 2500 2500>
    componentLinearAccScale=<1 1 1>
    componentLinearVelScale=<1 1 1>
    gravityOverride {z=-980}
    physicsBodyDefinitions [{
      boundBone {boneName=dyn_pants_l_front_02}
      boxExtents=<23 23 23>
      constraintSetup {angularLimitsMax=<5 15 0>, angularLimitsMin=<-5 -5 -20>, angularTarget=<1 0 0>}
      localJointOffset=<-4 0 0>
     }]
   }
   n36 = Anim Dynamics - Bone: dyn_pants_l_front_03!(SimulationSpace:AnimPhysSimSpaceType Enum, ComponentPose=@n35, Alpha=1) -> (Pose) {
    angularDampingOverride=0.9999
    angularSpringConstant=2000
    bAngularSpring=true
    bOverrideAngularDamping=true
    bUseGravityOverride=true
    gravityScale=0.35
    simulationSpace=RootRelative
    alphaBoolBlend {blendOption=Linear}
    boundBone {boneName=dyn_pants_l_front_03}
    componentAppliedLinearAccClamp=<2500 2500 2500>
    componentLinearAccScale=<1 1 1>
    componentLinearVelScale=<1 1 1>
    gravityOverride {z=-980}
    physicsBodyDefinitions [{
      boundBone {boneName=dyn_pants_l_front_03}
      boxExtents=<23 23 23>
      constraintSetup {angularLimitsMax=<5 15 0>, angularLimitsMin=<-5 -5 -20>, angularTarget=<1 0 0>}
      localJointOffset=<-4 0 0>
     }]
   }
   n37 = Anim Dynamics - Bone: dyn_pants_r_front_02!(SimulationSpace:AnimPhysSimSpaceType Enum, ComponentPose=@n36, Alpha=1) -> (Pose) {
    angularDampingOverride=0.9999
    angularSpringConstant=2000
    bAngularSpring=true
    bOverrideAngularDamping=true
    bUseGravityOverride=true
    gravityScale=0.35
    simulationSpace=RootRelative
    alphaBoolBlend {blendOption=Linear}
    boundBone {boneName=dyn_pants_r_front_02}
    componentAppliedLinearAccClamp=<2500 2500 2500>
    componentLinearAccScale=<1 1 1>
    componentLinearVelScale=<1 1 1>
    gravityOverride {z=-980}
    physicsBodyDefinitions [{
      boundBone {boneName=dyn_pants_r_front_02}
      boxExtents=<23 23 23>
      constraintSetup {angularLimitsMax=<5 15 0>, angularLimitsMin=<-5 -5 -20>, angularTarget=<1 0 0>}
      localJointOffset=<4 0 0>
     }]
   }
   n38 = Anim Dynamics - Bone: dyn_pants_r_front_03!(SimulationSpace:AnimPhysSimSpaceType Enum, ComponentPose=@n37, Alpha=1) -> (Pose) {
    angularDampingOverride=0.9999
    angularSpringConstant=2000
    bAngularSpring=true
    bOverrideAngularDamping=true
    bUseGravityOverride=true
    gravityScale=0.35
    simulationSpace=RootRelative
    alphaBoolBlend {blendOption=Linear}
    boundBone {boneName=dyn_pants_r_front_03}
    componentAppliedLinearAccClamp=<2500 2500 2500>
    componentLinearAccScale=<1 1 1>
    componentLinearVelScale=<1 1 1>
    gravityOverride {z=-980}
    physicsBodyDefinitions [{
      boundBone {boneName=dyn_pants_r_front_03}
      boxExtents=<23 23 23>
      constraintSetup {angularLimitsMax=<5 15 0>, angularLimitsMin=<-5 -5 -20>, angularTarget=<1 0 0>}
      localJointOffset=<4 0 0>
     }]
   }
   n39 = Anim Dynamics - Bone: dyn_pants_l_back_02!(SimulationSpace:AnimPhysSimSpaceType Enum, ComponentPose=@n38, Alpha=1) -> (Pose) {
    angularDampingOverride=0.9999
    angularSpringConstant=2000
    bAngularSpring=true
    bOverrideAngularDamping=true
    bUseGravityOverride=true
    gravityScale=0.35
    simulationSpace=RootRelative
    alphaBoolBlend {blendOption=Linear}
    boundBone {boneName=dyn_pants_l_back_02}
    componentAppliedLinearAccClamp=<2500 2500 2500>
    componentLinearAccScale=<1 1 1>
    componentLinearVelScale=<1 1 1>
    gravityOverride {z=-980}
    physicsBodyDefinitions [{
      boundBone {boneName=dyn_pants_l_back_02}
      boxExtents=<23 23 23>
      constraintSetup {angularLimitsMax=<5 15 35>, angularLimitsMin=<-5 -15 -3>, angularTarget=<1 0 0>}
      localJointOffset=<-4 0 0>
     }]
   }
   n40 = Anim Dynamics - Bone: dyn_pants_l_back_03!(SimulationSpace:AnimPhysSimSpaceType Enum, ComponentPose=@n39, Alpha=1) -> (Pose) {
    angularDampingOverride=0.9999
    angularSpringConstant=2000
    bAngularSpring=true
    bOverrideAngularDamping=true
    bUseGravityOverride=true
    gravityScale=0.35
    simulationSpace=RootRelative
    alphaBoolBlend {blendOption=Linear}
    boundBone {boneName=dyn_pants_l_back_03}
    componentAppliedLinearAccClamp=<2500 2500 2500>
    componentLinearAccScale=<1 1 1>
    componentLinearVelScale=<1 1 1>
    gravityOverride {z=-980}
    physicsBodyDefinitions [{
      boundBone {boneName=dyn_pants_l_back_03}
      boxExtents=<23 23 23>
      constraintSetup {angularLimitsMax=<5 15 35>, angularLimitsMin=<-5 -15 -3>, angularTarget=<1 0 0>}
      localJointOffset=<-4 0 0>
     }]
   }
   n41 = Anim Dynamics - Bone: dyn_pants_r_back_02!(SimulationSpace:AnimPhysSimSpaceType Enum, ComponentPose=@n40, Alpha=1) -> (Pose) {
    angularDampingOverride=0.9999
    angularSpringConstant=2000
    bAngularSpring=true
    bOverrideAngularDamping=true
    bUseGravityOverride=true
    gravityScale=0.35
    simulationSpace=RootRelative
    alphaBoolBlend {blendOption=Linear}
    boundBone {boneName=dyn_pants_r_back_02}
    componentAppliedLinearAccClamp=<2500 2500 2500>
    componentLinearAccScale=<1 1 1>
    componentLinearVelScale=<1 1 1>
    gravityOverride {z=-980}
    physicsBodyDefinitions [{
      boundBone {boneName=dyn_pants_r_back_02}
      boxExtents=<23 23 23>
      constraintSetup {angularLimitsMax=<5 15 35>, angularLimitsMin=<-5 -15 -3>, angularTarget=<1 0 0>}
      localJointOffset=<4 0 0>
     }]
   }
   n42 = Anim Dynamics - Bone: dyn_pants_r_back_03!(SimulationSpace:AnimPhysSimSpaceType Enum, ComponentPose=@n41, Alpha=1) -> (Pose) {
    angularDampingOverride=0.9999
    angularSpringConstant=2000
    bAngularSpring=true
    bOverrideAngularDamping=true
    bUseGravityOverride=true
    gravityScale=0.35
    simulationSpace=RootRelative
    alphaBoolBlend {blendOption=Linear}
    boundBone {boneName=dyn_pants_r_back_03}
    componentAppliedLinearAccClamp=<2500 2500 2500>
    componentLinearAccScale=<1 1 1>
    componentLinearVelScale=<1 1 1>
    gravityOverride {z=-980}
    physicsBodyDefinitions [{
      boundBone {boneName=dyn_pants_r_back_03}
      boxExtents=<23 23 23>
      constraintSetup {angularLimitsMax=<5 15 35>, angularLimitsMin=<-5 -15 -3>, angularTarget=<1 0 0>}
      localJointOffset=<4 0 0>
     }]
   }
   Output Pose(Result=Component To Local(ComponentPose=@n42)) {layerGroup=None}
  ]}
 EventGraph {nodes [
   Event BlueprintUpdateAnimation()
   > n1 = Gravity Override State Function DEPRECATED(CharacterPartAnimInstance=Self-Reference(), GameTimeInSeconds=GetGameTimeInSeconds(), GravityOverrideParams=Get Gravity Override Params())
   > Set GravityOverride(GravityOverride=@n1)
   > Set frontendAlpha(frontendAlpha:double=SelectFloat(A=0.5, B=1, bPickA=Get bIsInFrontEnd())) -> (Output_Get:double)
   > n2 = Branch(Condition=OR Boolean(A=Get bIsInFrontEnd(), B=Get bIsDBNOCarried()))
   > Set simSpaceChange(simSpaceChange:AnimPhysSimSpaceType Enum=World) -> (Output_Get:AnimPhysSimSpaceType Enum)
   Set simSpaceChange(execute=@n2.else)
  ]}
]
variables [
 frontendAlpha {type=number}
 Gravity Override Params {type=GravityOverrideParamsStruct}
 GravityOverride {type=Vector}
 simSpaceChange {type=AnimPhysSimSpaceType}
]
